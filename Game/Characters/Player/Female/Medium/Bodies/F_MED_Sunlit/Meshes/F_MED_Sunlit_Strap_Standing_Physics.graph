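# Physics asset for the standing strap mesh of the medium female player's Sunlit body, covering back straps, loops, dress panels, daggers and front ornaments.
# Capsules on pelvis, spine_01, thighs and calves form the body, and a large root sphere named floorCOL sits below as a floor.
# Small MSTR anchor bodies join pelvis, thighs or back straps with free angular motion, leading into angle-limited dangling bodies and chains.
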PhysicsAsset /Game/Characters/Player/Female/Medium/Bodies/F_MED_Sunlit/Meshes/F_MED_Sunlit_Strap_Standing_Physics
bodies {
 calf_l [{shapeType=Capsule, length=31.75, radius=11.75, shapeName=calf_l_capsule, center=<-27 0 0>, rotation=<rot 90 80.7734 75.7735>}]
 calf_r [{shapeType=Capsule, length=31.75, radius=11.75, shapeName=calf_r_capsule, center=<27 0 0>, rotation=<rot 90 80.7735 75.7736>}]
 dyn_back_strap_l_01 [{shapeType=Capsule, length=3, radius=1.5, shapeName=dyn_back_strap_l_01_capsule, center=<2 -1 0>, rotation=<rot 90 0 0>}]
 dyn_back_strap_l_02 [{shapeType=Capsule, length=9.0137, radius=2.2363, shapeName=dyn_back_strap_l_02_capsule, center=<8 -2 0>, rotation=<rot 90 0 0>}]
 dyn_back_strap_l_03 [
  {shapeType=Capsule, length=13.4448, radius=3.5552, shapeName=dyn_back_strap_l_03_capsule, center=<10 0 0>, rotation=<rot 90 0 0>}
  {
   shapeType=Capsule
   length=21.9171
   radius=4.7271
   shapeName=dyn_back_strap_l_03_capsule
   center=<10.3278 -11.7104 33.2414>
   rotation=<rot 56.5778 -87.4673 0>
  }
 ]
 dyn_back_strap_l_04 [{shapeType=Capsule, length=9.079, radius=3.9708, shapeName=dyn_back_strap_l_04_capsule, center=<6.8438 0 0>, rotation=<rot 90 0 0>}]
 dyn_back_strap_l_05 [{shapeType=Capsule, length=8.173, radius=2, shapeName=dyn_back_strap_l_04_capsule, center=<5.2324 0 0>, rotation=<rot 90 0 0>}]
 dyn_back_strap_l_MSTR [{shapeType=Capsule, length=0.505, radius=0.505, shapeName=dyn_back_strap_l_MSTR_capsule, rotation=<rot 0 -90 0>}]
 dyn_back_strap_r_01 [{shapeType=Capsule, length=3, radius=1.5, shapeName=dyn_back_strap_r_01_capsule, center=<2 0 0>, rotation=<rot 90 0 0>}]
 dyn_back_strap_r_02 [{shapeType=Capsule, length=8.1691, radius=2.9559, shapeName=dyn_back_strap_r_02_capsule, center=<9 0 0>, rotation=<rot 90 0 0>}]
 dyn_back_strap_r_03 [
  {shapeType=Capsule, length=11.0521, radius=2.9479, shapeName=dyn_back_strap_r_03_capsule, center=<10 0 0>, rotation=<rot 90 0 0>}
  {
   shapeType=Capsule
   length=29.9024
   radius=5.6254
   shapeName=dyn_back_strap_r_03_capsule
   center=<13.3086 1.479 -25.7108>
   rotation=<rot 2.0133 -180 -120.1631>
  }
 ]
 dyn_back_strap_r_04 [{shapeType=Capsule, length=3.8425, radius=5.1427, shapeName=dyn_back_strap_r_04_capsule, center=<5.4498 0 0>, rotation=<rot 90 0 0>}]
 dyn_back_strap_r_05 [{shapeType=Capsule, length=7.7145, radius=2, shapeName=dyn_back_strap_r_05_capsule, center=<5.3843 0 0>, rotation=<rot 90 0 0>}]
 dyn_back_strap_r_MSTR [{shapeType=Capsule, length=0.505, radius=0.505, shapeName=dyn_back_strap_r_MSTR_capsule, rotation=<rot 0 -90 0>}]
 dyn_l_BK_Dress [{shapeType=Capsule, length=10.3836, radius=4.0958, shapeName=dyn_l_BK_Dress_capsule, center=<17.1353 0 2>, rotation=<rot 0 0 90>}]
 dyn_l_BK_Dress_MSTR [{shapeType=Capsule, length=0.505, radius=0.505, shapeName=dyn_l_BK_Dress_MSTR_capsule, rotation=<rot 0 -90 0>}]
 dyn_l_FR_Dress [{shapeType=Capsule, length=11.6129, radius=2.8664, shapeName=dyn_l_FR_Dress_capsule, center=<14.6073 0 2>, rotation=<rot 0 0 90>}]
 dyn_l_FR_Dress_MSTR [{shapeType=Capsule, length=0.505, radius=0.505, shapeName=dyn_l_FR_Dress_MSTR_capsule, rotation=<rot 0 -90 0>}]
 dyn_l_Side_Dress [{shapeType=Capsule, length=7.9974, radius=6.482, shapeName=dyn_l_Side_Dress_capsule, center=<14.6073 0 2>, rotation=<rot 0 0 90>}]
 dyn_l_Side_Dress_MSTR [{shapeType=Capsule, length=0.505, radius=0.505, shapeName=dyn_l_Side_Dress_MSTR_capsule, rotation=<rot 0 -90 0>}]
 dyn_l_dagger [{shapeType=Capsule, length=36.9093, radius=2, shapeName=dyn_l_dagger_capsule, center=<8.6996 0 0>, rotation=<rot 90 0 0>}]
 dyn_l_dagger_MSTR [{shapeType=Capsule, length=0.505, radius=0.505, shapeName=dyn_l_dagger_MSTR_capsule, rotation=<rot 0 -90 0>}]
 dyn_l_front_ring_01 [
  {shapeType=Capsule, length=10.8062, radius=3.6732, shapeName=dyn_r_FR_Dress_capsule, center=<9.0663 0 5>, rotation=<rot 0 0 90>}
  {
   shapeType=Capsule
   length=25.8741
   radius=4.5046
   shapeName=dyn_l_front_ring_01_capsule
   center=<8.338 -6.9165 -7.4174>
   rotation=<rot 0 0 15.5217>
  }
 ]
 dyn_l_front_ring_02 [
  {shapeType=Capsule, length=10.8062, radius=3.6732, shapeName=dyn_r_FR_Dress_capsule, center=<9.0663 0 5>, rotation=<rot 0 0 90>}
  {
   shapeType=Capsule
   length=30.2206
   radius=4.4912
   shapeName=dyn_l_front_ring_02_capsule
   center=<11.5833 -3.6622 -8.6063>
   rotation=<rot 0 0 15.0643>
  }
 ]
 dyn_l_front_square [
  {shapeType=Capsule, length=10.8062, radius=3.6732, shapeName=dyn_r_FR_Dress_capsule, center=<9.0663 0 5>, rotation=<rot 0 0 90>}
  {
   shapeType=Capsule
   length=30.9811
   radius=4.4505
   shapeName=dyn_l_front_square_capsule
   center=<11.4565 -14.1611 -11.0159>
   rotation=<rot 0.0054 0.0026 21.7374>
  }
 ]
 dyn_loop_l_01 [{shapeType=Capsule, length=5.7756, radius=4.2244, shapeName=dyn_loop_l_01_capsule, center=<5.2967 0 4>, rotation=<rot 0 0 90>}]
 dyn_loop_l_02 [
  {shapeType=Capsule, length=4.0902, radius=8.033, shapeName=dyn_loop_l_02_capsule, center=<7.7309 0 0>, rotation=<rot 0 0 90>}
  {
   shapeType=Capsule
   length=1.6262
   radius=8.3738
   shapeName=dyn_loop_l_02_capsule
   center=<1.074 8.1999 0.3555>
   rotation=<rot -87.8201 -119.6797 109.4485>
  }
 ]
 dyn_loop_l_MSTR [{shapeType=Capsule, length=0.505, radius=0.505, shapeName=dyn_loop_l_MSTR_capsule, rotation=<rot 0 -90 0>}]
 dyn_loop_l_bk [{shapeType=Capsule, length=8.3746, radius=5.6254, shapeName=dyn_back_strap_r_03_capsule, center=<15.5758 0 0>, rotation=<rot 90 0 0>}]
 dyn_loop_l_bk_MSTR [{shapeType=Capsule, length=0.505, radius=0.505, shapeName=dyn_loop_l_bk_MSTR_capsule, rotation=<rot 0 -90 0>}]
 dyn_loop_l_fr [
  {shapeType=Capsule, length=10.3698, radius=4.1095, shapeName=dyn_r_FR_Dress_capsule, center=<9.0663 0 5>, rotation=<rot 0 0 90>}
  {
   shapeType=Capsule
   length=26.6988
   radius=4.6333
   shapeName=dyn_loop_l_fr_capsule
   center=<10.8442 -14.1519 -7.3146>
   rotation=<rot 0.0001 0 18.2786>
  }
 ]
 dyn_loop_l_fr_MSTR [{shapeType=Capsule, length=0.505, radius=0.505, shapeName=dyn_loop_l_fr_MSTR_capsule, rotation=<rot 0 -90 0>}]
 dyn_loop_r_01 [
  {shapeType=Capsule, length=19.9175, radius=6.2683, shapeName=dyn_loop_l_01_capsule, center=<5.6977 -6.5 0>}
  {shapeType=Capsule, length=2.4786, radius=7.5214, shapeName=dyn_loop_r_01_capsule, center=<8.9445 2.5693 16.3724>, rotation=<rot 0 0 90>}
 ]
 dyn_loop_r_02 [{shapeType=Capsule, length=28.3917, radius=8.8331, shapeName=dyn_loop_l_02_capsule, center=<7.7309 -6.5 0>}]
 dyn_loop_r_MSTR [{shapeType=Capsule, length=0.505, radius=0.505, shapeName=dyn_loop_r_MSTR_capsule, rotation=<rot 0 -90 0>}]
 dyn_loop_r_bk [{shapeType=Capsule, length=8.3746, radius=5.6254, shapeName=dyn_back_strap_r_03_capsule, center=<15.5758 0 0>, rotation=<rot 90 0 0>}]
 dyn_loop_r_bk_MSTR [{shapeType=Capsule, length=0.505, radius=0.505, shapeName=dyn_loop_r_bk_MSTR_capsule, rotation=<rot 0 -90 0>}]
 dyn_loop_r_fr [
  {shapeType=Capsule, length=9.4253, radius=5.0541, shapeName=dyn_r_FR_Dress_capsule, center=<20.2188 0 5>, rotation=<rot 0 0 90>}
  {
   shapeType=Capsule
   length=19.8294
   radius=5.4328
   shapeName=dyn_loop_r_fr_capsule
   center=<23.9895 9.1552 -6.3698>
   rotation=<rot 0 0 154.6536>
  }
 ]
 dyn_loop_r_fr_MSTR [{
   shapeType=Capsule
   length=9.446
   radius=3.3442
   shapeName=dyn_loop_r_fr_MSTR_capsule
   center=<11.8619 -0.9389 0.8005>
   rotation=<rot 0 87.8271 -89.631>
  }]
 dyn_r_BK_Dress [{shapeType=Capsule, length=10.3836, radius=4.0958, shapeName=dyn_r_BK_Dress_capsule, center=<17.1353 0 2>, rotation=<rot 0 0 90>}]
 dyn_r_BK_Dress_MSTR [{shapeType=Capsule, length=0.505, radius=0.505, shapeName=dyn_r_BK_Dress_MSTR_capsule, rotation=<rot 0 -90 0>}]
 dyn_r_FR_Dress [{shapeType=Capsule, length=10.2837, radius=4.1956, shapeName=dyn_r_FR_Dress_capsule, center=<14.6073 0 2>, rotation=<rot 0 0 90>}]
 dyn_r_FR_Dress_MSTR [{shapeType=Capsule, length=0.505, radius=0.505, shapeName=dyn_r_FR_Dress_MSTR_capsule, rotation=<rot 0 -90 0>}]
 dyn_r_Side_Dress [{shapeType=Capsule, length=9.1616, radius=5.3178, shapeName=dyn_r_Side_Dress_capsule, center=<14.6073 0 2>, rotation=<rot 0 0 90>}]
 dyn_r_Side_Dress_MSTR [{shapeType=Capsule, length=0.505, radius=0.505, shapeName=dyn_r_Side_Dress_MSTR_capsule, rotation=<rot 0 -90 0>}]
 dyn_r_dagger [{shapeType=Capsule, length=36.9093, radius=2, shapeName=dyn_r_dagger_capsule, center=<8.6996 0 0>, rotation=<rot 90 0 0>}]
 dyn_r_dagger_MSTR [{shapeType=Capsule, length=0.505, radius=0.505, shapeName=dyn_r_dagger_MSTR_capsule, rotation=<rot 0 -90 0>}]
 pelvis [
  {shapeType=Capsule, length=10.0096, radius=9.25}
  {shapeType=Capsule, length=20.1875, radius=7.0329, shapeName=pelvis_capsule, center=<-13.2387 -2.8952 0>}
 ]
 root [{shapeType=Sphere, radius=183.6686, shapeName=floorCOL_sphere, center=<-3.4923 -2.7141 -174.8455>}]
 spine_01 [{shapeType=Capsule, length=10, radius=8, shapeName=spine_01_capsule, rotation=<rot 0 -90 0>}]
 thigh_l [
  {shapeType=Capsule, length=17.1637, radius=9.3227, center=<-21.2717 -0.3685 1.6327>, rotation=<rot 90 89.4327 89.433>}
  {shapeType=Capsule, length=2.9854, radius=9.1396, shapeName=thigh_r_capsule, center=<-39.4665 0.7099 1.7776>, rotation=<rot 90 0 0>}
 ]
 thigh_r [
  {shapeType=Capsule, length=13.23, radius=13.2564, center=<21.2717 -0.3685 -1.6327>, rotation=<rot 90 89.4327 89.433>}
  {shapeType=Capsule, length=1.9962, radius=10.1288, shapeName=thigh_r_capsule, center=<39.4665 0.7099 -1.7776>, rotation=<rot 90 0 0>}
 ]
 thigh_twist_01_l [{shapeType=Capsule, length=23.5, radius=11, shapeName=thigh_twist_01_l_capsule, center=<8 0 0>, rotation=<rot 90 0 0>}]
 thigh_twist_01_r [{shapeType=Capsule, length=24.8905, radius=9.6095, shapeName=thigh_twist_01_r_capsule, center=<-8 0 0>, rotation=<rot 90 0 0>}]
}
constraints [
 {
  bone1Name=calf_l
  bone2Name=thigh_l
  swing1LimitDegrees=45
  swing1Motion=Free
  swing2LimitDegrees=45
  swing2Motion=Free
  twistLimitDegrees=45
  twistMotion=Free
 }
 {
  bone1Name=calf_r
  bone2Name=thigh_r
  swing1LimitDegrees=45
  swing1Motion=Free
  swing2LimitDegrees=45
  swing2Motion=Free
  twistLimitDegrees=45
  twistMotion=Free
 }
 {
  bone1Name=dyn_loop_l_MSTR
  bone2Name=thigh_l
  swing1LimitDegrees=45
  swing1Motion=Free
  swing2LimitDegrees=45
  swing2Motion=Free
  twistLimitDegrees=45
  twistMotion=Free
 }
 {
  bone1Name=dyn_loop_r_MSTR
  bone2Name=thigh_r
  swing1LimitDegrees=45
  swing1Motion=Free
  swing2LimitDegrees=45
  swing2Motion=Free
  twistLimitDegrees=45
  twistMotion=Free
 }
 {bone1Name=dyn_loop_l_01, bone2Name=dyn_loop_l_MSTR, swing1LimitDegrees=35, swing2LimitDegrees=35, twistLimitDegrees=3}
 {bone1Name=dyn_loop_l_02, bone2Name=dyn_loop_l_01, swing1LimitDegrees=35, swing2LimitDegrees=35, twistLimitDegrees=3}
 {
  bone1Name=thigh_l
  bone2Name=pelvis
  swing1LimitDegrees=45
  swing1Motion=Free
  swing2LimitDegrees=45
  swing2Motion=Free
  twistLimitDegrees=45
  twistMotion=Free
 }
 {
  bone1Name=thigh_r
  bone2Name=pelvis
  swing1LimitDegrees=45
  swing1Motion=Free
  swing2LimitDegrees=45
  swing2Motion=Free
  twistLimitDegrees=45
  twistMotion=Free
 }
 {bone1Name=dyn_loop_r_01, bone2Name=dyn_loop_r_MSTR, swing1LimitDegrees=35, swing2LimitDegrees=35, twistLimitDegrees=3}
 {bone1Name=dyn_loop_r_02, bone2Name=dyn_loop_r_01, swing1LimitDegrees=35, swing2LimitDegrees=35, twistLimitDegrees=3}
 {
  bone1Name=thigh_twist_01_r
  bone2Name=thigh_r
  swing1LimitDegrees=45
  swing1Motion=Free
  swing2LimitDegrees=45
  swing2Motion=Free
  twistLimitDegrees=45
  twistMotion=Free
 }
 {
  bone1Name=thigh_twist_01_l
  bone2Name=thigh_l
  swing1LimitDegrees=45
  swing1Motion=Free
  swing2LimitDegrees=45
  swing2Motion=Free
  twistLimitDegrees=45
  twistMotion=Free
 }
 {
  bone1Name=dyn_back_strap_l_MSTR
  bone2Name=pelvis
  swing1LimitDegrees=45
  swing1Motion=Free
  swing2LimitDegrees=45
  swing2Motion=Free
  twistLimitDegrees=45
  twistMotion=Free
 }
 {
  bone1Name=dyn_back_strap_r_MSTR
  bone2Name=pelvis
  swing1LimitDegrees=45
  swing1Motion=Free
  swing2LimitDegrees=45
  swing2Motion=Free
  twistLimitDegrees=45
  twistMotion=Free
 }
 {bone1Name=dyn_back_strap_l_01, bone2Name=dyn_back_strap_l_MSTR, swing1LimitDegrees=20, swing2LimitDegrees=20, twistLimitDegrees=1}
 {bone1Name=dyn_back_strap_l_02, bone2Name=dyn_back_strap_l_01, swing1LimitDegrees=15, swing2LimitDegrees=15, twistLimitDegrees=15}
 {bone1Name=dyn_back_strap_l_03, bone2Name=dyn_back_strap_l_02, swing1LimitDegrees=35, swing2LimitDegrees=35, twistLimitDegrees=15}
 {bone1Name=dyn_back_strap_l_04, bone2Name=dyn_back_strap_l_03, swing1LimitDegrees=35, swing2LimitDegrees=35, twistLimitDegrees=15}
 {bone1Name=dyn_back_strap_l_05, bone2Name=dyn_back_strap_l_04, swing1LimitDegrees=35, swing2LimitDegrees=35, twistLimitDegrees=15}
 {bone1Name=dyn_back_strap_r_01, bone2Name=dyn_back_strap_r_MSTR, swing1LimitDegrees=20, swing2LimitDegrees=20, twistLimitDegrees=1}
 {bone1Name=dyn_back_strap_r_02, bone2Name=dyn_back_strap_r_01, swing1LimitDegrees=15, swing2LimitDegrees=15, twistLimitDegrees=15}
 {bone1Name=dyn_back_strap_r_03, bone2Name=dyn_back_strap_r_02, swing1LimitDegrees=35, swing2LimitDegrees=35, twistLimitDegrees=15}
 {bone1Name=dyn_back_strap_r_04, bone2Name=dyn_back_strap_r_03, swing1LimitDegrees=35, swing2LimitDegrees=35, twistLimitDegrees=15}
 {bone1Name=dyn_back_strap_r_05, bone2Name=dyn_back_strap_r_04, swing1LimitDegrees=35, swing2LimitDegrees=35, twistLimitDegrees=15}
 {
  bone1Name=pelvis
  bone2Name=root
  swing1LimitDegrees=45
  swing1Motion=Free
  swing2LimitDegrees=45
  swing2Motion=Free
  twistLimitDegrees=45
  twistMotion=Free
 }
 {
  bone1Name=dyn_l_BK_Dress_MSTR
  bone2Name=pelvis
  swing1LimitDegrees=45
  swing1Motion=Free
  swing2LimitDegrees=45
  swing2Motion=Free
  twistLimitDegrees=45
  twistMotion=Free
 }
 {
  bone1Name=dyn_r_BK_Dress_MSTR
  bone2Name=pelvis
  swing1LimitDegrees=45
  swing1Motion=Free
  swing2LimitDegrees=45
  swing2Motion=Free
  twistLimitDegrees=45
  twistMotion=Free
 }
 {
  bone1Name=dyn_l_FR_Dress_MSTR
  bone2Name=thigh_l
  swing1LimitDegrees=45
  swing1Motion=Free
  swing2LimitDegrees=45
  swing2Motion=Free
  twistLimitDegrees=45
  twistMotion=Free
 }
 {
  bone1Name=dyn_l_Side_Dress_MSTR
  bone2Name=thigh_l
  swing1LimitDegrees=45
  swing1Motion=Free
  swing2LimitDegrees=45
  swing2Motion=Free
  twistLimitDegrees=45
  twistMotion=Free
 }
 {
  bone1Name=dyn_r_FR_Dress_MSTR
  bone2Name=thigh_r
  swing1LimitDegrees=45
  swing1Motion=Free
  swing2LimitDegrees=45
  swing2Motion=Free
  twistLimitDegrees=45
  twistMotion=Free
 }
 {
  bone1Name=dyn_r_Side_Dress_MSTR
  bone2Name=thigh_r
  swing1LimitDegrees=45
  swing1Motion=Free
  swing2LimitDegrees=45
  swing2Motion=Free
  twistLimitDegrees=45
  twistMotion=Free
 }
 {bone1Name=dyn_l_BK_Dress, bone2Name=dyn_l_BK_Dress_MSTR, swing1LimitDegrees=30, swing2LimitDegrees=15, twistLimitDegrees=5}
 {bone1Name=dyn_r_BK_Dress, bone2Name=dyn_r_BK_Dress_MSTR, swing1LimitDegrees=30, swing2LimitDegrees=15, twistLimitDegrees=5}
 {bone1Name=dyn_l_FR_Dress, bone2Name=dyn_l_FR_Dress_MSTR, swing1LimitDegrees=30, swing2LimitDegrees=30, twistLimitDegrees=5}
 {bone1Name=dyn_l_Side_Dress, bone2Name=dyn_l_Side_Dress_MSTR, swing1LimitDegrees=30, swing2LimitDegrees=15, twistLimitDegrees=5}
 {bone1Name=dyn_r_FR_Dress, bone2Name=dyn_r_FR_Dress_MSTR, swing1LimitDegrees=30, swing2LimitDegrees=30, twistLimitDegrees=5}
 {bone1Name=dyn_r_Side_Dress, bone2Name=dyn_r_Side_Dress_MSTR, swing1LimitDegrees=30, swing2LimitDegrees=15, twistLimitDegrees=5}
 {
  bone1Name=dyn_l_dagger_MSTR
  bone2Name=pelvis
  swing1LimitDegrees=45
  swing1Motion=Free
  swing2LimitDegrees=45
  swing2Motion=Free
  twistLimitDegrees=45
  twistMotion=Free
 }
 {
  bone1Name=dyn_r_dagger_MSTR
  bone2Name=pelvis
  swing1LimitDegrees=45
  swing1Motion=Free
  swing2LimitDegrees=45
  swing2Motion=Free
  twistLimitDegrees=45
  twistMotion=Free
 }
 {bone1Name=dyn_l_dagger, bone2Name=dyn_l_dagger_MSTR, swing1LimitDegrees=10, swing2LimitDegrees=25, twistLimitDegrees=5}
 {bone1Name=dyn_r_dagger, bone2Name=dyn_r_dagger_MSTR, swing1LimitDegrees=10, swing2LimitDegrees=25, twistLimitDegrees=5}
 {
  bone1Name=dyn_loop_r_fr_MSTR
  bone2Name=thigh_r
  swing1LimitDegrees=45
  swing1Motion=Free
  swing2LimitDegrees=45
  swing2Motion=Free
  twistLimitDegrees=45
  twistMotion=Free
 }
 {
  bone1Name=dyn_loop_l_fr_MSTR
  bone2Name=thigh_l
  swing1LimitDegrees=45
  swing1Motion=Free
  swing2LimitDegrees=45
  swing2Motion=Free
  twistLimitDegrees=45
  twistMotion=Free
 }
 {bone1Name=dyn_loop_l_fr, bone2Name=dyn_loop_l_fr_MSTR, swing1LimitDegrees=35, swing2LimitDegrees=35, twistLimitDegrees=5}
 {bone1Name=dyn_loop_r_fr, bone2Name=dyn_loop_r_fr_MSTR, swing1LimitDegrees=20, swing2LimitDegrees=20, twistLimitDegrees=20}
 {
  bone1Name=spine_01
  bone2Name=pelvis
  swing1LimitDegrees=45
  swing1Motion=Free
  swing2LimitDegrees=45
  swing2Motion=Free
  twistLimitDegrees=45
  twistMotion=Free
 }
 {bone1Name=dyn_l_front_square, bone2Name=dyn_loop_l_fr, swing1LimitDegrees=35, swing2LimitDegrees=35, twistLimitDegrees=5}
 {bone1Name=dyn_l_front_ring_01, bone2Name=dyn_l_front_square, swing1LimitDegrees=35, swing2LimitDegrees=35, twistLimitDegrees=5}
 {bone1Name=dyn_l_front_ring_02, bone2Name=dyn_l_front_ring_01, swing1LimitDegrees=35, swing2LimitDegrees=35, twistLimitDegrees=5}
 {
  bone1Name=dyn_loop_l_bk_MSTR
  bone2Name=dyn_back_strap_l_02
  swing1LimitDegrees=45
  swing1Motion=Free
  swing2LimitDegrees=45
  swing2Motion=Free
  twistLimitDegrees=45
  twistMotion=Free
 }
 {
  bone1Name=dyn_loop_r_bk_MSTR
  bone2Name=dyn_back_strap_r_02
  swing1LimitDegrees=45
  swing1Motion=Free
  swing2LimitDegrees=45
  swing2Motion=Free
  twistLimitDegrees=45
  twistMotion=Free
 }
 {bone1Name=dyn_loop_l_bk, bone2Name=dyn_loop_l_bk_MSTR, swing1LimitDegrees=30, swing2LimitDegrees=30, twistLimitDegrees=5}
 {bone1Name=dyn_loop_r_bk, bone2Name=dyn_loop_r_bk_MSTR, swing1LimitDegrees=30, swing2LimitDegrees=30, twistLimitDegrees=5}
]
