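# Post-process style AnimBP for the F_MED_Sunlit female body, adding secondary motion to its straps, loops and necklace.
# Branches from cached PreSim pose run RigidBody with standing, crouching or frontend physics assets, then constrain front loops and rings to thigh bones.
# A non-RigidBody fallback uses bone offsets, trail controllers and AnimDynamics.
# Bools for rigid-body enablement, crouch and frontend select the branch, which feeds necklace AnimDynamics.
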AnimBlueprint /Game/Characters/Player/Female/Medium/Bodies/F_MED_Sunlit/Meshes/F_MED_Sunlit_AnimBP
target_skeleton=&/F_MED_Sunlit_Skeleton
graphs [
 AnimGraph {nodes [
   n1 = RigidBody!(
    ComponentPose=Local To Component(LocalPose=Use cached pose 'PreSim'())
    bAlphaBoolEnabled:bool
    OverrideWorldGravity:Vector="0,0,-980"
    ExternalForce
   ) -> (Pose) {
    alphaInputType=Bool
    bClampLinearTranslationLimitToRefPose=true
    bOverrideWorldGravity=true
    overridePhysicsAsset=&/F_MED_Sunlit_Strap_Standing_Physics
    simulationSpace=BaseBoneSpace
    alphaBoolBlend {blendOption=Linear}
    baseBoneRef {boneName=pelvis}
    componentAppliedLinearAccClamp=<2500 2500 2500>
    componentLinearAccScale=<0.3 0.3 0.4>
    componentLinearVelScale=<0.3 0.3 0.4>
    overrideWorldGravity {z=-980}
    simSpaceSettings {
     maxAngularAcceleration=1000
     maxAngularVelocity=3
     maxLinearAcceleration=700
     maxLinearVelocity=500
     velocityScaleZ=0.5
     worldAlpha=1
     externalLinearDragV=<0.2 0.2>
    }
   }
   Blend Poses by bool(BlendPose_0=Component To Local(ComponentPose=@n1), bActiveValue=Get bIsPlayingEmote()) {blendType=Linear}
   n2 = "Constraint - dyn_loop_l_MSTR (R)"!(ComponentPose, bAlphaBoolEnabled:bool=True, ConstraintWeights_0=0.5) -> (Pose) {
    alphaInputType=Bool
    alphaBoolBlend {blendInTime=0.1, blendOption=Linear, blendOutTime=0.1}
    boneToModify {boneName=dyn_loop_l_MSTR}
    constraintSetup [{transformType=Rotation, targetBone {boneName=thigh_twist_01_l}}]
    constraintWeights [1]
   }
   n3 = "Constraint - dyn_loop_r_MSTR (R)"!(ComponentPose=@n2, bAlphaBoolEnabled:bool=True, ConstraintWeights_0=0.5) -> (Pose) {
    alphaInputType=Bool
    alphaBoolBlend {blendInTime=0.1, blendOption=Linear, blendOutTime=0.1}
    boneToModify {boneName=dyn_loop_r_MSTR}
    constraintSetup [{transformType=Rotation, targetBone {boneName=thigh_twist_01_r}}]
    constraintWeights [1]
   }
   n4 = "Constraint - dyn_loop_l_01 (R)"!(ComponentPose=@n3, bAlphaBoolEnabled:bool=True, ConstraintWeights_0=0.5) -> (Pose) {
    alphaInputType=Bool
    alphaBoolBlend {blendInTime=0.1, blendOption=Linear, blendOutTime=0.1}
    boneToModify {boneName=dyn_loop_l_01}
    constraintSetup [{transformType=Rotation, targetBone {boneName=thigh_twist_01_l}}]
    constraintWeights [1]
   }
   n5 = "Constraint - dyn_loop_r_01 (R)"!(ComponentPose=@n4, bAlphaBoolEnabled:bool=True, ConstraintWeights_0=0.5) -> (Pose) {
    alphaInputType=Bool
    alphaBoolBlend {blendInTime=0.1, blendOption=Linear, blendOutTime=0.1}
    boneToModify {boneName=dyn_loop_r_01}
    constraintSetup [{transformType=Rotation, targetBone {boneName=thigh_twist_01_r}}]
    constraintWeights [1]
   }
   n6 = "Constraint - dyn_loop_l_02 (R)"!(ComponentPose=@n5, bAlphaBoolEnabled:bool=True, ConstraintWeights_0=0.5) -> (Pose) {
    alphaInputType=Bool
    alphaBoolBlend {blendInTime=0.1, blendOption=Linear, blendOutTime=0.1}
    boneToModify {boneName=dyn_loop_l_02}
    constraintSetup [{transformType=Rotation, targetBone {boneName=thigh_twist_01_l}}]
    constraintWeights [1]
   }
   "Constraint - dyn_loop_r_02 (R)"!(ComponentPose=@n6, bAlphaBoolEnabled:bool=True, ConstraintWeights_0=0.5) -> (Pose) {
    alphaInputType=Bool
    alphaBoolBlend {blendInTime=0.1, blendOption=Linear, blendOutTime=0.1}
    boneToModify {boneName=dyn_loop_r_02}
    constraintSetup [{transformType=Rotation, targetBone {boneName=thigh_twist_01_r}}]
    constraintWeights [1]
   }
   n7 = "Transform (Modify) Bone - Bone: dyn_loop_l_fr_MSTR"(Alpha=Get ThighRotL() -> (ThighRotL:double), Translation="0,0,0", Rotation="15,-15,0", Scale="1,1,1") {
    rotationMode=BMM_Additive
    rotationSpace=BCS_BoneSpace
    translationMode=BMM_Additive
    translationSpace=BCS_BoneSpace
    alphaBoolBlend {blendInTime=0.1, blendOption=Linear, blendOutTime=0.1}
    alphaScaleBiasClamp {bInterpResult=true}
    boneToModify {boneName=dyn_loop_l_fr_MSTR}
    rotation {pitch=15, yaw=-15}
   }
   "Transform (Modify) Bone - Bone: dyn_loop_r_fr_MSTR"(ComponentPose=@n7, Alpha=Get ThighRotR() -> (ThighRotR:double), Translation="0,0,0", Rotation="20,-15,0", Scale="1,1,1") {
    rotationMode=BMM_Additive
    rotationSpace=BCS_BoneSpace
    translationMode=BMM_Additive
    translationSpace=BCS_BoneSpace
    alphaBoolBlend {blendInTime=0.1, blendOption=Linear, blendOutTime=0.1}
    alphaScaleBiasClamp {bInterpResult=true}
    boneToModify {boneName=dyn_loop_r_fr_MSTR}
    rotation {pitch=20, yaw=-15}
   }
   n8 = RigidBody(ComponentPose=Local To Component(LocalPose=Use cached pose 'PreSim'()), OverrideWorldGravity="0,0,-980") {
    alphaInputType=Bool
    bClampLinearTranslationLimitToRefPose=true
    bOverrideWorldGravity=true
    overridePhysicsAsset=&F_MED_Sunlit_Strap_Standing_Physics
    simulationSpace=BaseBoneSpace
    alphaBoolBlend {blendOption=Linear}
    baseBoneRef {boneName=pelvis}
    componentAppliedLinearAccClamp=<5000 5000 5000>
    componentLinearAccScale=<0.3 0.3 0.4>
    componentLinearVelScale=<0.3 0.3 0.4>
    overrideWorldGravity {z=-980}
    simSpaceSettings {
     maxAngularAcceleration=1000
     maxAngularVelocity=3
     maxLinearAcceleration=700
     maxLinearVelocity=500
     velocityScaleZ=0.5
     worldAlpha=1
     externalLinearDragV=<0.2 0.2>
    }
   }
   n9 = "Constraint - dyn_loop_l_fr (R)"!(ComponentPose=@n8, bAlphaBoolEnabled:bool=True, ConstraintWeights_0=0.1) -> (Pose) {
    alphaInputType=Bool
    alphaBoolBlend {blendOption=Linear, blendOutTime=0.1}
    boneToModify {boneName=dyn_loop_l_fr}
    constraintSetup [{transformType=Rotation, targetBone {boneName=thigh_l}}]
    constraintWeights [1]
   }
   n10 = "Constraint - dyn_loop_r_fr_MSTR (P)"!(ComponentPose=@n9, bAlphaBoolEnabled:bool=True, ConstraintWeights_0=1) -> (Pose) {
    alphaInputType=Bool
    alphaBoolBlend {blendOption=Linear, blendOutTime=0.1}
    boneToModify {boneName=dyn_loop_r_fr_MSTR}
    constraintSetup [{targetBone {boneName=thigh_r}}]
    constraintWeights [1]
   }
   n11 = "Constraint - dyn_loop_r_fr (P)"!(ComponentPose=@n10, bAlphaBoolEnabled:bool=True, ConstraintWeights_0=1) -> (Pose) {
    alphaInputType=Bool
    alphaBoolBlend {blendOption=Linear, blendOutTime=0.1}
    boneToModify {boneName=dyn_loop_r_fr}
    constraintSetup [{targetBone {boneName=thigh_r}}]
    constraintWeights [1]
   }
   n12 = "Transform (Modify) Bone - Bone: dyn_loop_r_fr_MSTR"!(ComponentPose=@n11, bAlphaBoolEnabled:bool=True, Translation="3.082,-9.808,-1.431", Rotation="8.480,59.683,21.765", Scale="1,1,1") -> (Pose) {
    alphaInputType=Bool
    rotationMode=BMM_Additive
    rotationSpace=BCS_BoneSpace
    scaleMode=BMM_Additive
    scaleSpace=BCS_BoneSpace
    translationMode=BMM_Additive
    translationSpace=BCS_BoneSpace
    alphaBoolBlend {blendInTime=0.2, blendOption=Linear, blendOutTime=0.2}
    boneToModify {boneName=dyn_loop_r_fr_MSTR}
    rotation=<rot 8.48 59.683 21.765>
    translation=<3.082 -9.808 -1.431>
   }
   n13 = "Constraint - dyn_loop_r_fr_MSTR (PPPPPP)"!(
    ComponentPose=@n12
    bAlphaBoolEnabled:bool=True
    ConstraintWeights_0=0.4364
    ConstraintWeights_1:float=0.034
    ConstraintWeights_2:float=0.03
    ConstraintWeights_3:float=0.3093
    ConstraintWeights_4:float=0.0921
    ConstraintWeights_5:float=0.0981
   ) -> (Pose) {
    alphaInputType=Bool
    alphaBoolBlend {blendOption=Linear}
    boneToModify {boneName=dyn_loop_r_fr_MSTR}
    constraintSetup [
     {targetBone {boneName=thigh_twist_01_r}}
     {targetBone {boneName=thigh_r}}
     {targetBone {boneName=dyn_r_Side_Dress}}
     {targetBone {boneName=pelvis}}
     {targetBone {boneName=dyn_r_FR_Dress}}
     {targetBone {boneName=deform_groin_r}}
    ]
    constraintWeights [1,1,1,1,1,1]
   }
   n14 = "Constraint - dyn_loop_l_fr_MSTR (PPPPPP)"!(
    ComponentPose=@n13
    bAlphaBoolEnabled:bool=True
    ConstraintWeights_0=0.0921
    ConstraintWeights_1:float=0.034
    ConstraintWeights_2:float=0.3093
    ConstraintWeights_3:float=0.4364
    ConstraintWeights_4:float=0.03
    ConstraintWeights_5:float=0.0981
   ) -> (Pose) {
    alphaInputType=Bool
    alphaBoolBlend {blendOption=Linear}
    boneToModify {boneName=dyn_loop_l_fr_MSTR}
    constraintSetup [
     {targetBone {boneName=dyn_l_FR_Dress}}
     {targetBone {boneName=thigh_l}}
     {targetBone {boneName=pelvis}}
     {targetBone {boneName=thigh_twist_01_l}}
     {targetBone {boneName=dyn_l_Side_Dress}}
     {targetBone {boneName=deform_groin_l}}
    ]
    constraintWeights [1,1,1,1,1,1]
   }
   n15 = "Constraint - dyn_loop_l_fr (P)"!(ComponentPose=@n14, bAlphaBoolEnabled:bool=True, ConstraintWeights_0=1) -> (Pose) {
    alphaInputType=Bool
    alphaBoolBlend {blendInTime=0.1, blendOption=Linear, blendOutTime=0.1}
    boneToModify {boneName=dyn_loop_l_fr}
    constraintSetup [{targetBone {boneName=thigh_twist_01_l}}]
    constraintWeights [1]
   }
   n16 = "Constraint - dyn_loop_r_fr (P)"(ComponentPose=@n15, bAlphaBoolEnabled=True, ConstraintWeights_0=1) {
    alphaInputType=Bool
    alphaBoolBlend {blendInTime=0.1, blendOption=Linear, blendOutTime=0.1}
    boneToModify {boneName=dyn_loop_r_fr}
    constraintSetup [{targetBone {boneName=thigh_twist_01_r}}]
    constraintWeights [1]
   }
   n17 = "Constraint - dyn_l_front_square (P)"!(ComponentPose=@n16, bAlphaBoolEnabled:bool=True, ConstraintWeights_0=1) -> (Pose) {
    alphaInputType=Bool
    alphaBoolBlend {blendInTime=0.1, blendOption=Linear, blendOutTime=0.1}
    boneToModify {boneName=dyn_l_front_square}
    constraintSetup [{targetBone {boneName=thigh_twist_01_l}}]
    constraintWeights [1]
   }
   n18 = "Constraint - dyn_r_front_square (P)"!(ComponentPose=@n17, bAlphaBoolEnabled:bool=True, ConstraintWeights_0=1) -> (Pose) {
    alphaInputType=Bool
    alphaBoolBlend {blendInTime=0.1, blendOption=Linear, blendOutTime=0.1}
    boneToModify {boneName=dyn_r_front_square}
    constraintSetup [{targetBone {boneName=thigh_twist_01_r}}]
    constraintWeights [1]
   }
   n19 = "Constraint - dyn_l_front_ring_01 (P)"!(ComponentPose=@n18, bAlphaBoolEnabled:bool=True, ConstraintWeights_0=1) -> (Pose) {
    alphaInputType=Bool
    alphaBoolBlend {blendInTime=0.1, blendOption=Linear, blendOutTime=0.1}
    boneToModify {boneName=dyn_l_front_ring_01}
    constraintSetup [{targetBone {boneName=thigh_twist_01_l}}]
    constraintWeights [1]
   }
   n20 = "Constraint - dyn_r_front_ring_01 (P)"!(ComponentPose=@n19, bAlphaBoolEnabled:bool=True, ConstraintWeights_0=1) -> (Pose) {
    alphaInputType=Bool
    alphaBoolBlend {blendInTime=0.1, blendOption=Linear, blendOutTime=0.1}
    boneToModify {boneName=dyn_r_front_ring_01}
    constraintSetup [{targetBone {boneName=thigh_twist_01_r}}]
    constraintWeights [1]
   }
   n21 = "Constraint - dyn_l_front_ring_02 (P)"!(ComponentPose=@n20, bAlphaBoolEnabled:bool=True, ConstraintWeights_0=1) -> (Pose) {
    alphaInputType=Bool
    alphaBoolBlend {blendInTime=0.1, blendOption=Linear, blendOutTime=0.1}
    boneToModify {boneName=dyn_l_front_ring_02}
    constraintSetup [{targetBone {boneName=thigh_twist_01_l}}]
    constraintWeights [1]
   }
   n22 = "Constraint - dyn_r_front_ring_02 (P)"!(ComponentPose=@n21, bAlphaBoolEnabled:bool=True, ConstraintWeights_0=1) -> (Pose) {
    alphaInputType=Bool
    alphaBoolBlend {blendInTime=0.1, blendOption=Linear, blendOutTime=0.1}
    boneToModify {boneName=dyn_r_front_ring_02}
    constraintSetup [{targetBone {boneName=thigh_twist_01_r}}]
    constraintWeights [1]
   }
   n23 = "Constraint - dyn_loop_r_MSTR (R)"!(ComponentPose=Local To Component(LocalPose=Use cached pose 'PreSim'()), Alpha:float=1, ConstraintWeights_0=1) -> (Pose) {
    alphaBoolBlend {blendOption=Linear}
    boneToModify {boneName=dyn_loop_r_MSTR}
    constraintSetup [{transformType=Rotation, targetBone {boneName=thigh_r}}]
    constraintWeights [1]
   }
   n24 = RigidBody(ComponentPose=@n23, OverrideWorldGravity="0,0,-980") {
    alphaInputType=Bool
    bClampLinearTranslationLimitToRefPose=true
    bOverrideWorldGravity=true
    overridePhysicsAsset=&/F_MED_Sunlit_Strap_Frontend_Physics1
    simulationSpace=BaseBoneSpace
    alphaBoolBlend {blendOption=Linear}
    baseBoneRef {boneName=pelvis}
    componentAppliedLinearAccClamp=<5000 5000 5000>
    componentLinearAccScale=<0.3 0.3 0.4>
    componentLinearVelScale=<0.3 0.3 0.4>
    overrideWorldGravity {z=-980}
    simSpaceSettings {
     maxAngularAcceleration=1000
     maxAngularVelocity=3
     maxLinearAcceleration=700
     maxLinearVelocity=500
     velocityScaleZ=0.5
     worldAlpha=1
     externalLinearDragV=<0.2 0.2>
    }
   }
   n25 = "Constraint - dyn_loop_l_fr_MSTR (P)"!(ComponentPose=@n24, bAlphaBoolEnabled:bool=True, ConstraintWeights_0=1) -> (Pose) {
    alphaInputType=Bool
    alphaBoolBlend {blendInTime=0.1, blendOption=Linear, blendOutTime=0.1}
    boneToModify {boneName=dyn_loop_l_fr_MSTR}
    constraintSetup [{targetBone {boneName=thigh_twist_01_l}}]
    constraintWeights [1]
   }
   n26 = "Constraint - dyn_loop_r_fr_MSTR (P)"(ComponentPose=@n25, bAlphaBoolEnabled=True, ConstraintWeights_0=1) {
    alphaInputType=Bool
    alphaBoolBlend {blendInTime=0.1, blendOption=Linear, blendOutTime=0.1}
    boneToModify {boneName=dyn_loop_r_fr_MSTR}
    constraintSetup [{targetBone {boneName=thigh_twist_01_r}}]
    constraintWeights [1]
   }
   n27 = "Constraint - dyn_loop_l_fr (P)"(ComponentPose=@n26, bAlphaBoolEnabled=True, ConstraintWeights_0=1) {
    alphaInputType=Bool
    alphaBoolBlend {blendInTime=0.1, blendOption=Linear, blendOutTime=0.1}
    boneToModify {boneName=dyn_loop_l_fr}
    constraintSetup [{targetBone {boneName=thigh_twist_01_l}}]
    constraintWeights [1]
   }
   n28 = "Constraint - dyn_loop_r_fr (P)"(ComponentPose=@n27, bAlphaBoolEnabled=True, ConstraintWeights_0=1) {
    alphaInputType=Bool
    alphaBoolBlend {blendInTime=0.1, blendOption=Linear, blendOutTime=0.1}
    boneToModify {boneName=dyn_loop_r_fr}
    constraintSetup [{targetBone {boneName=thigh_twist_01_r}}]
    constraintWeights [1]
   }
   n29 = "Constraint - dyn_l_front_square (P)"(ComponentPose=@n28, bAlphaBoolEnabled=True, ConstraintWeights_0=1) {
    alphaInputType=Bool
    alphaBoolBlend {blendInTime=0.1, blendOption=Linear, blendOutTime=0.1}
    boneToModify {boneName=dyn_l_front_square}
    constraintSetup [{targetBone {boneName=thigh_twist_01_l}}]
    constraintWeights [1]
   }
   n30 = "Constraint - dyn_r_front_square (P)"(ComponentPose=@n29, bAlphaBoolEnabled=True, ConstraintWeights_0=1) {
    alphaInputType=Bool
    alphaBoolBlend {blendInTime=0.1, blendOption=Linear, blendOutTime=0.1}
    boneToModify {boneName=dyn_r_front_square}
    constraintSetup [{targetBone {boneName=thigh_twist_01_r}}]
    constraintWeights [1]
   }
   n31 = "Constraint - dyn_l_front_ring_01 (P)"(ComponentPose=@n30, bAlphaBoolEnabled=True, ConstraintWeights_0=1) {
    alphaInputType=Bool
    alphaBoolBlend {blendInTime=0.1, blendOption=Linear, blendOutTime=0.1}
    boneToModify {boneName=dyn_l_front_ring_01}
    constraintSetup [{targetBone {boneName=thigh_twist_01_l}}]
    constraintWeights [1]
   }
   n32 = "Constraint - dyn_r_front_ring_01 (P)"(ComponentPose=@n31, bAlphaBoolEnabled=True, ConstraintWeights_0=1) {
    alphaInputType=Bool
    alphaBoolBlend {blendInTime=0.1, blendOption=Linear, blendOutTime=0.1}
    boneToModify {boneName=dyn_r_front_ring_01}
    constraintSetup [{targetBone {boneName=thigh_twist_01_r}}]
    constraintWeights [1]
   }
   n33 = "Constraint - dyn_l_front_ring_02 (P)"(ComponentPose=@n32, bAlphaBoolEnabled=True, ConstraintWeights_0=1) {
    alphaInputType=Bool
    alphaBoolBlend {blendInTime=0.1, blendOption=Linear, blendOutTime=0.1}
    boneToModify {boneName=dyn_l_front_ring_02}
    constraintSetup [{targetBone {boneName=thigh_twist_01_l}}]
    constraintWeights [1]
   }
   n34 = "Constraint - dyn_r_front_ring_02 (P)"(ComponentPose=@n33, bAlphaBoolEnabled=True, ConstraintWeights_0=1) {
    alphaInputType=Bool
    alphaBoolBlend {blendInTime=0.1, blendOption=Linear, blendOutTime=0.1}
    boneToModify {boneName=dyn_r_front_ring_02}
    constraintSetup [{targetBone {boneName=thigh_twist_01_r}}]
    constraintWeights [1]
   }
   n35 = "Constraint - dyn_loop_r_fr_MSTR (PPPPPP)"(
    ComponentPose=Local To Component(LocalPose=Use cached pose 'PreSim'())
    bAlphaBoolEnabled=True
    ConstraintWeights_0=0.4364
    ConstraintWeights_1=0.034
    ConstraintWeights_2=0.03
    ConstraintWeights_3=0.3093
    ConstraintWeights_4=0.0921
    ConstraintWeights_5=0.0981
   ) {
    alphaInputType=Bool
    alphaBoolBlend {blendOption=Linear}
    boneToModify {boneName=dyn_loop_r_fr_MSTR}
    constraintSetup [
     {targetBone {boneName=thigh_twist_01_r}}
     {targetBone {boneName=thigh_r}}
     {targetBone {boneName=dyn_r_Side_Dress}}
     {targetBone {boneName=pelvis}}
     {targetBone {boneName=dyn_r_FR_Dress}}
     {targetBone {boneName=deform_groin_r}}
    ]
    constraintWeights [1,1,1,1,1,1]
   }
   n36 = "Constraint - dyn_loop_l_fr_MSTR (PPPPPP)"(
    ComponentPose=@n35
    bAlphaBoolEnabled=True
    ConstraintWeights_0=0.0921
    ConstraintWeights_1=0.034
    ConstraintWeights_2=0.3093
    ConstraintWeights_3=0.4364
    ConstraintWeights_4=0.03
    ConstraintWeights_5=0.0981
   ) {
    alphaInputType=Bool
    alphaBoolBlend {blendOption=Linear}
    boneToModify {boneName=dyn_loop_l_fr_MSTR}
    constraintSetup [
     {targetBone {boneName=dyn_l_FR_Dress}}
     {targetBone {boneName=thigh_l}}
     {targetBone {boneName=pelvis}}
     {targetBone {boneName=thigh_twist_01_l}}
     {targetBone {boneName=dyn_l_Side_Dress}}
     {targetBone {boneName=deform_groin_l}}
    ]
    constraintWeights [1,1,1,1,1,1]
   }
   n37 = "Constraint - dyn_loop_l_fr_MSTR (P)"(ComponentPose=@n36, bAlphaBoolEnabled=True, ConstraintWeights_0=1) {
    alphaInputType=Bool
    alphaBoolBlend {blendOption=Linear}
    boneToModify {boneName=dyn_loop_l_fr_MSTR}
    constraintSetup [{targetBone {boneName=thigh_twist_01_l}}]
    constraintWeights [1]
   }
   n38 = "Constraint - dyn_loop_r_fr_MSTR (P)"(ComponentPose=@n37, bAlphaBoolEnabled=True, ConstraintWeights_0=1) {
    alphaInputType=Bool
    alphaBoolBlend {blendOption=Linear}
    boneToModify {boneName=dyn_loop_r_fr_MSTR}
    constraintSetup [{targetBone {boneName=thigh_twist_01_r}}]
    constraintWeights [1]
   }
   n39 = "Transform (Modify) Bone - Bone: dyn_loop_l_MSTR"!(ComponentPose=@n38, bAlphaBoolEnabled:bool=True, Translation="0,0,0", Rotation="12.918,0,6.857", Scale="1,1,1") -> (Pose) {
    alphaInputType=Bool
    rotationMode=BMM_Additive
    rotationSpace=BCS_BoneSpace
    scaleMode=BMM_Additive
    scaleSpace=BCS_BoneSpace
    translationMode=BMM_Additive
    translationSpace=BCS_BoneSpace
    alphaBoolBlend {blendOption=Linear}
    boneToModify {boneName=dyn_loop_l_MSTR}
    rotation {pitch=12.918, roll=6.857}
   }
   n40 = "Transform (Modify) Bone - Bone: dyn_loop_r_MSTR"!(ComponentPose=@n39, bAlphaBoolEnabled:bool=True, Translation="0,0,0", Rotation="0.017,-10.619,0.028", Scale="1,1,1") -> (Pose) {
    alphaInputType=Bool
    rotationMode=BMM_Additive
    rotationSpace=BCS_BoneSpace
    scaleMode=BMM_Additive
    scaleSpace=BCS_BoneSpace
    translationMode=BMM_Additive
    translationSpace=BCS_BoneSpace
    alphaBoolBlend {blendOption=Linear}
    boneToModify {boneName=dyn_loop_r_MSTR}
    rotation=<rot 0.017 -10.619 0.028>
   }
   n41 = "Transform (Modify) Bone - Bone: dyn_back_strap_l_MSTR"(ComponentPose=@n40, Alpha=1, Translation="0,0,0", Rotation="-11.178,-6.336,0.112", Scale="1,1,1") {
    rotationMode=BMM_Additive
    rotationSpace=BCS_BoneSpace
    scaleMode=BMM_Additive
    scaleSpace=BCS_BoneSpace
    translationMode=BMM_Additive
    translationSpace=BCS_BoneSpace
    alphaBoolBlend {blendOption=Linear}
    boneToModify {boneName=dyn_back_strap_l_MSTR}
    rotation=<rot -11.178 -6.3359 0.112>
   }
   n42 = "Transform (Modify) Bone - Bone: dyn_back_strap_r_MSTR"(ComponentPose=@n41, Alpha=1, Translation="0,0,0", Rotation="6.398,-11.334,-0.080", Scale="1,1,1") {
    rotationMode=BMM_Additive
    rotationSpace=BCS_BoneSpace
    scaleMode=BMM_Additive
    scaleSpace=BCS_BoneSpace
    translationMode=BMM_Additive
    translationSpace=BCS_BoneSpace
    alphaBoolBlend {blendOption=Linear}
    boneToModify {boneName=dyn_back_strap_r_MSTR}
    rotation=<rot 6.398 -11.3339 -0.08>
   }
   n43 = "Transform (Modify) Bone - Bone: dyn_loop_l_fr_MSTR"!(ComponentPose=@n42, bAlphaBoolEnabled:bool=True, Translation="0,0,0", Rotation="0,0,0", Scale="1,1,1") -> (Pose) {
    alphaInputType=Bool
    rotationMode=BMM_Additive
    rotationSpace=BCS_BoneSpace
    translationMode=BMM_Additive
    translationSpace=BCS_BoneSpace
    alphaBoolBlend {blendInTime=0.1, blendOption=Linear, blendOutTime=0.1}
    boneToModify {boneName=dyn_loop_l_fr_MSTR}
   }
   n44 = "Transform (Modify) Bone - Bone: dyn_loop_r_fr_MSTR"(ComponentPose=@n43, bAlphaBoolEnabled=True, Translation="0,0,1.5", Rotation="10,0,0", Scale="1,1,1") {
    alphaInputType=Bool
    rotationMode=BMM_Additive
    rotationSpace=BCS_BoneSpace
    translationMode=BMM_Additive
    translationSpace=BCS_BoneSpace
    alphaBoolBlend {blendInTime=0.1, blendOption=Linear, blendOutTime=0.1}
    boneToModify {boneName=dyn_loop_r_fr_MSTR}
    rotation {pitch=10}
    translation {z=1.5}
   }
   n45 = "Transform (Modify) Bone - Bone: dyn_back_strap_l_02"!(ComponentPose=@n44, bAlphaBoolEnabled:bool=True, Translation="0,0,0", Rotation="-15,15,0", Scale="1,1,1") -> (Pose) {
    alphaInputType=Bool
    rotationMode=BMM_Additive
    rotationSpace=BCS_BoneSpace
    translationMode=BMM_Additive
    translationSpace=BCS_BoneSpace
    alphaBoolBlend {blendInTime=0.1, blendOption=Linear, blendOutTime=0.1}
    boneToModify {boneName=dyn_back_strap_l_02}
    rotation {pitch=-15, yaw=15}
   }
   n46 = "Transform (Modify) Bone - Bone: dyn_back_strap_r_02"!(ComponentPose=@n45, bAlphaBoolEnabled:bool=True, Translation="0,0,0", Rotation="15,15,0", Scale="1,1,1") -> (Pose) {
    alphaInputType=Bool
    rotationMode=BMM_Additive
    rotationSpace=BCS_BoneSpace
    translationMode=BMM_Additive
    translationSpace=BCS_BoneSpace
    alphaBoolBlend {blendInTime=0.1, blendOption=Linear, blendOutTime=0.1}
    boneToModify {boneName=dyn_back_strap_r_02}
    rotation {pitch=15, yaw=15}
   }
   n47 = "Transform (Modify) Bone - Bone: dyn_loop_r_MSTR"(ComponentPose=@n46, bAlphaBoolEnabled=True, Translation="0,0,0", Rotation="0,-30,0", Scale="1,1,1") {
    alphaInputType=Bool
    rotationMode=BMM_Additive
    rotationSpace=BCS_BoneSpace
    translationMode=BMM_Additive
    translationSpace=BCS_BoneSpace
    alphaBoolBlend {blendInTime=0.1, blendOption=Linear, blendOutTime=0.1}
    boneToModify {boneName=dyn_loop_r_MSTR}
    rotation {yaw=-30}
   }
   n48 = Get still_crouch() -> (still_crouch:bool)
   n49 = "Transform (Modify) Bone - Bone: dyn_back_strap_l_02"(ComponentPose=@n47, bAlphaBoolEnabled=@n48, Translation="0,0,0", Rotation="-15.482,27.607,-10.321", Scale="1,1,1") {
    alphaInputType=Bool
    rotationMode=BMM_Additive
    rotationSpace=BCS_BoneSpace
    translationMode=BMM_Additive
    translationSpace=BCS_BoneSpace
    alphaBoolBlend {blendInTime=0.1, blendOption=Linear, blendOutTime=0.1}
    boneToModify {boneName=dyn_back_strap_l_02}
    rotation=<rot -15.482 27.607 -10.321>
   }
   n50 = "Transform (Modify) Bone - Bone: dyn_back_strap_l_03"!(ComponentPose=@n49, bAlphaBoolEnabled:bool=@n48, Translation="0,0,0", Rotation="15.988,18.575,14.607", Scale="1,1,1") -> (Pose) {
    alphaInputType=Bool
    rotationMode=BMM_Additive
    rotationSpace=BCS_BoneSpace
    translationMode=BMM_Additive
    translationSpace=BCS_BoneSpace
    alphaBoolBlend {blendInTime=0.1, blendOption=Linear, blendOutTime=0.1}
    boneToModify {boneName=dyn_back_strap_l_03}
    rotation=<rot 15.988 18.575 14.607>
   }
   n51 = "Transform (Modify) Bone - Bone: dyn_back_strap_l_04"!(ComponentPose=@n50, bAlphaBoolEnabled:bool=@n48, Translation="0,0,0", Rotation="0,-30,0", Scale="1,1,1") -> (Pose) {
    alphaInputType=Bool
    rotationMode=BMM_Additive
    rotationSpace=BCS_BoneSpace
    translationMode=BMM_Additive
    translationSpace=BCS_BoneSpace
    alphaBoolBlend {blendInTime=0.1, blendOption=Linear, blendOutTime=0.1}
    boneToModify {boneName=dyn_back_strap_l_04}
    rotation {yaw=-30}
   }
   n52 = "Transform (Modify) Bone - Bone: dyn_back_strap_r_02"(ComponentPose=@n51, bAlphaBoolEnabled=@n48, Translation="0,0,0", Rotation="-1.157,16.693,2.050", Scale="1,1,1") {
    alphaInputType=Bool
    rotationMode=BMM_Additive
    rotationSpace=BCS_BoneSpace
    translationMode=BMM_Additive
    translationSpace=BCS_BoneSpace
    alphaBoolBlend {blendInTime=0.1, blendOption=Linear, blendOutTime=0.1}
    boneToModify {boneName=dyn_back_strap_r_02}
    rotation=<rot -1.157 16.693 2.05>
   }
   n53 = "Transform (Modify) Bone - Bone: dyn_back_strap_r_03"!(ComponentPose=@n52, bAlphaBoolEnabled:bool=@n48, Translation="0,0,0", Rotation="-3.289,23.266,-2.553", Scale="1,1,1") -> (Pose) {
    alphaInputType=Bool
    rotationMode=BMM_Additive
    rotationSpace=BCS_BoneSpace
    translationMode=BMM_Additive
    translationSpace=BCS_BoneSpace
    alphaBoolBlend {blendInTime=0.1, blendOption=Linear, blendOutTime=0.1}
    boneToModify {boneName=dyn_back_strap_r_03}
    rotation=<rot -3.289 23.266 -2.553>
   }
   n54 = "Transform (Modify) Bone - Bone: dyn_back_strap_r_04"!(ComponentPose=@n53, bAlphaBoolEnabled:bool=@n48, Translation="0,0,0", Rotation="0,0,0", Scale="1,1,1") -> (Pose) {
    alphaInputType=Bool
    rotationMode=BMM_Additive
    rotationSpace=BCS_BoneSpace
    translationMode=BMM_Additive
    translationSpace=BCS_BoneSpace
    alphaBoolBlend {blendInTime=0.1, blendOption=Linear, blendOutTime=0.1}
    boneToModify {boneName=dyn_back_strap_r_04}
   }
   n55 = "Transform (Modify) Bone - Bone: dyn_loop_r_fr_MSTR"(ComponentPose=@n54, bAlphaBoolEnabled=True, Translation="0,0,0", Rotation="5,0,0", Scale="1,1,1") {
    alphaInputType=Bool
    rotationMode=BMM_Additive
    rotationSpace=BCS_BoneSpace
    translationMode=BMM_Additive
    translationSpace=BCS_BoneSpace
    alphaBoolBlend {blendInTime=0.1, blendOption=Linear, blendOutTime=0.1}
    boneToModify {boneName=dyn_loop_r_fr_MSTR}
    rotation {pitch=5}
   }
   n56 = "Transform (Modify) Bone - Bone: dyn_back_strap_l_01"(ComponentPose=@n55, Alpha=Get ThighRotL(), Translation="0,0,0", Rotation="15,-15,0", Scale="1,1,1") {
    rotationMode=BMM_Additive
    rotationSpace=BCS_BoneSpace
    translationMode=BMM_Additive
    translationSpace=BCS_BoneSpace
    alphaBoolBlend {blendInTime=0.1, blendOption=Linear, blendOutTime=0.1}
    alphaScaleBiasClamp {bInterpResult=true}
    boneToModify {boneName=dyn_back_strap_l_01}
    rotation {pitch=15, yaw=-15}
   }
   n57 = "Transform (Modify) Bone - Bone: dyn_back_strap_r_01"(ComponentPose=@n56, Alpha=Get ThighRotR(), Translation="0,0,0", Rotation="-15,-15,0", Scale="1,1,1") {
    rotationMode=BMM_Additive
    rotationSpace=BCS_BoneSpace
    translationMode=BMM_Additive
    translationSpace=BCS_BoneSpace
    alphaBoolBlend {blendInTime=0.1, blendOption=Linear, blendOutTime=0.1}
    alphaScaleBiasClamp {bInterpResult=true}
    boneToModify {boneName=dyn_back_strap_r_01}
    rotation {pitch=-15, yaw=-15}
   }
   n58 = Get BwdMovement() -> (BwdMovement:float)
   n59 = Get RelaxationAlpha() -> (RelaxationAlpha:double)
   n60 = Trail controller - Bone: dyn_back_strap_r_05!(ComponentPose=@n57, Alpha=@n58, RelaxationSpeedScale:float=@n59, FakeVelocity:Vector="0,0,0") -> (Pose) {
    bActorSpaceFakeVel=true
    bShowBaseMotion=false
    bShowLimit=false
    chainLength=4
    debugLifeTime=0.1
    maxDeltaTime=0.0333
    alphaScaleBiasClamp {bInterpResult=true}
    baseJoint {boneName=root}
    planarLimits [
     {drivingBone {boneName=root}, planeTransform {location=<0 0 10>, rotation=<rot 0 0 0>, scale=<1 1 1>}}
     {drivingBone {boneName=spine_03}, planeTransform {location=<0 -20 0>, rotation=<rot 0 15 -89.9999>, scale=<1 1 1>}}
    ]
    rotationLimits [{limitMax=<180 180 180>}, {limitMax=<180 180 180>}, {limitMax=<180 180 180>}, {limitMax=<180 180 180>}]
    rotationOffsets [<0 0 0>, <0 0 0>, <0 0 0>, <0 0 0>]
    trailBone {boneName=dyn_back_strap_r_05}
    trailRelaxationSpeed {editorCurveData {keys [{arriveTangent=-72.0739, interpMode=RCIM_Cubic, leaveTangent=-72.0739, tangentMode=RCTM_User, time=0, value=35}, <at 1 5 RCIM_Cubic>]}}
   }
   n61 = Trail controller - Bone: dyn_back_strap_l_04!(ComponentPose=@n60, Alpha=@n58, RelaxationSpeedScale:float=@n59, FakeVelocity:Vector="0,0,0") -> (Pose) {
    bActorSpaceFakeVel=true
    bShowBaseMotion=false
    bShowLimit=false
    chainLength=3
    debugLifeTime=0.1
    maxDeltaTime=0.0333
    alphaScaleBiasClamp {bInterpResult=true}
    baseJoint {boneName=root}
    planarLimits [
     {drivingBone {boneName=root}, planeTransform {location=<0 0 10>, rotation=<rot 0 0 0>, scale=<1 1 1>}}
     {drivingBone {boneName=spine_03}, planeTransform {location=<0 -20 0>, rotation=<rot 0 15 -89.9999>, scale=<1 1 1>}}
    ]
    rotationLimits [{limitMax=<180 180 180>}, {limitMax=<180 180 180>}, {limitMax=<180 180 180>}]
    rotationOffsets [<0 0 0>, <0 0 0>, <0 0 0>]
    trailBone {boneName=dyn_back_strap_l_04}
    trailRelaxationSpeed {editorCurveData {keys [{arriveTangent=-72.0739, interpMode=RCIM_Cubic, leaveTangent=-72.0739, tangentMode=RCTM_User, time=0, value=35}, <at 1 5 RCIM_Cubic>]}}
   }
   n62 = Anim Dynamics - Bone: dyn_loop_l_01(ComponentPose=@n61, Alpha=1, ExternalForce="0,0,0") {
    angularDampingOverride=0.9999
    angularSpringConstant=150
    bAngularSpring=true
    bGravityOverrideInSimSpace=true
    bOverrideAngularDamping=true
    simulationSpace=RootRelative
    alphaBoolBlend {blendOption=Linear}
    boundBone {boneName=dyn_loop_l_01}
    componentAppliedLinearAccClamp=<3500 3500 3500>
    componentLinearAccScale=<3 3 3>
    componentLinearVelScale=<3 3 3>
    physicsBodyDefinitions [{
      sphereCollisionRadius=0
      boundBone {boneName=dyn_loop_l_01}
      boxExtents=<10 20 20>
      constraintSetup {angularLimitsMax=<5 5 15>, angularLimitsMin=<-5 -35 -15>, angularTarget=<0 0 0>}
      localJointOffset=<2 0 -0.25>
     }]
   }
   n63 = Anim Dynamics - Bone: dyn_loop_l_02(ComponentPose=@n62, Alpha=1, ExternalForce="0,0,0") {
    angularDampingOverride=0.9999
    angularSpringConstant=150
    bAngularSpring=true
    bGravityOverrideInSimSpace=true
    bOverrideAngularDamping=true
    simulationSpace=RootRelative
    alphaBoolBlend {blendOption=Linear}
    boundBone {boneName=dyn_loop_l_02}
    componentAppliedLinearAccClamp=<3500 3500 3500>
    componentLinearAccScale=<3 3 3>
    componentLinearVelScale=<3 3 3>
    physicsBodyDefinitions [{
      sphereCollisionRadius=0
      boundBone {boneName=dyn_loop_l_02}
      boxExtents=<10 20 20>
      constraintSetup {angularLimitsMax=<5 5 15>, angularLimitsMin=<-5 -35 -15>, angularTarget=<0 0 0>}
      localJointOffset=<2 0 -0.25>
     }]
   }
   n64 = Anim Dynamics - Bone: dyn_loop_r_01(ComponentPose=@n63, Alpha=1, ExternalForce="0,0,0") {
    angularDampingOverride=0.9999
    angularSpringConstant=150
    bAngularSpring=true
    bGravityOverrideInSimSpace=true
    bOverrideAngularDamping=true
    simulationSpace=RootRelative
    alphaBoolBlend {blendOption=Linear}
    boundBone {boneName=dyn_loop_r_01}
    componentAppliedLinearAccClamp=<3500 3500 3500>
    componentLinearAccScale=<3 3 3>
    componentLinearVelScale=<3 3 3>
    physicsBodyDefinitions [{
      sphereCollisionRadius=0
      boundBone {boneName=dyn_loop_r_01}
      boxExtents=<10 20 20>
      constraintSetup {angularLimitsMax=<5 15 5>, angularLimitsMin=<-5 -15 -35>, angularTarget=<0 0 0>}
      localJointOffset=<2 0 -0.25>
     }]
   }
   n65 = Anim Dynamics - Bone: dyn_loop_r_02(ComponentPose=@n64, Alpha=1, ExternalForce="0,0,0") {
    angularDampingOverride=0.9999
    angularSpringConstant=150
    bAngularSpring=true
    bGravityOverrideInSimSpace=true
    bOverrideAngularDamping=true
    simulationSpace=RootRelative
    alphaBoolBlend {blendOption=Linear}
    boundBone {boneName=dyn_loop_r_02}
    componentAppliedLinearAccClamp=<3500 3500 3500>
    componentLinearAccScale=<3 3 3>
    componentLinearVelScale=<3 3 3>
    physicsBodyDefinitions [{
      sphereCollisionRadius=0
      boundBone {boneName=dyn_loop_r_02}
      boxExtents=<10 20 20>
      constraintSetup {angularLimitsMax=<5 15 5>, angularLimitsMin=<-5 -15 -35>, angularTarget=<0 0 0>}
      localJointOffset=<2 0 -0.25>
     }]
   }
   n66 = Blend Poses by bool(
    BlendPose_0=Blend Poses by bool(BlendPose_0=Component To Local(ComponentPose=@n34), BlendPose_1=Component To Local(ComponentPose=@n22), bActiveValue=Get bIsInFrontEnd()) {blendType=Linear}
    BlendPose_1=Component To Local(ComponentPose=@n65)
    bActiveValue=Get bRigidBodyEnabled()
   ) {blendType=Linear}
   n67 = "Constraint - dyn_loop_r_fr_MSTR (R)"(ComponentPose=Local To Component(LocalPose=Use cached pose 'PreSim'()), Alpha=1, ConstraintWeights_0=1) {
    alphaBoolBlend {blendOption=Linear}
    boneToModify {boneName=dyn_loop_r_fr_MSTR}
    constraintSetup [{transformType=Rotation, targetBone {boneName=thigh_twist_01_r}}]
    constraintWeights [1]
   }
   n68 = "Transform (Modify) Bone - Bone: dyn_loop_r_fr_MSTR"(ComponentPose=@n67, bAlphaBoolEnabled=True, Translation="-0.672,-1.170,-0.025", Rotation="0.068,60.179,19.550", Scale="1,1,1") {
    alphaInputType=Bool
    rotationMode=BMM_Additive
    rotationSpace=BCS_BoneSpace
    scaleMode=BMM_Additive
    scaleSpace=BCS_BoneSpace
    translationMode=BMM_Additive
    translationSpace=BCS_BoneSpace
    alphaBoolBlend {blendInTime=0.2, blendOption=Linear, blendOutTime=0.2}
    boneToModify {boneName=dyn_loop_r_fr_MSTR}
    rotation=<rot 0.068 60.179 19.55>
    translation=<-0.672 -1.17 -0.025>
   }
   n69 = "Transform (Modify) Bone - Bone: dyn_loop_r_MSTR"(ComponentPose=@n68, bAlphaBoolEnabled=True, Translation="0,0,0", Rotation="32.011,-18.658,-1.636", Scale="1,1,1") {
    alphaInputType=Bool
    rotationMode=BMM_Additive
    rotationSpace=BCS_BoneSpace
    scaleMode=BMM_Additive
    scaleSpace=BCS_BoneSpace
    translationMode=BMM_Additive
    translationSpace=BCS_BoneSpace
    alphaBoolBlend {blendInTime=0.2, blendOption=Linear, blendOutTime=0.2}
    boneToModify {boneName=dyn_loop_r_MSTR}
    rotation=<rot 32.011 -18.658 -1.6359>
   }
   n70 = RigidBody(ComponentPose=@n69, OverrideWorldGravity="0,0,-980") {
    alphaInputType=Bool
    bClampLinearTranslationLimitToRefPose=true
    bOverrideWorldGravity=true
    overridePhysicsAsset=&/F_MED_Sunlit_Strap_Crouching_Physics
    simulationSpace=BaseBoneSpace
    alphaBoolBlend {blendOption=Linear}
    baseBoneRef {boneName=pelvis}
    componentAppliedLinearAccClamp=<5000 5000 5000>
    componentLinearAccScale=<0.3 0.3 0.4>
    componentLinearVelScale=<0.3 0.3 0.4>
    overrideWorldGravity {z=-980}
    simSpaceSettings {
     maxAngularAcceleration=1000
     maxAngularVelocity=3
     maxLinearAcceleration=700
     maxLinearVelocity=500
     velocityScaleZ=0.5
     worldAlpha=1
     externalLinearDragV=<0.2 0.2>
    }
   }
   n71 = "Transform (Modify) Bone - Bone: dyn_l_front_ring_01"!(ComponentPose=@n70, bAlphaBoolEnabled:bool=True, Translation="0,0,0", Rotation="-33,-0.057,8.722", Scale="1,1,1") -> (Pose) {
    alphaInputType=Bool
    rotationMode=BMM_Additive
    rotationSpace=BCS_BoneSpace
    scaleMode=BMM_Additive
    scaleSpace=BCS_BoneSpace
    translationMode=BMM_Additive
    translationSpace=BCS_BoneSpace
    alphaBoolBlend {blendInTime=0.1, blendOption=Linear, blendOutTime=0.1}
    boneToModify {boneName=dyn_l_front_ring_01}
    rotation=<rot -33 -0.057 8.722>
   }
   n72 = "Constraint - dyn_loop_l_fr_MSTR (P)"(ComponentPose=@n71, bAlphaBoolEnabled=True, ConstraintWeights_0=1) {
    alphaInputType=Bool
    alphaBoolBlend {blendInTime=0.1, blendOption=Linear, blendOutTime=0.1}
    boneToModify {boneName=dyn_loop_l_fr_MSTR}
    constraintSetup [{targetBone {boneName=thigh_twist_01_l}}]
    constraintWeights [1]
   }
   n73 = "Constraint - dyn_loop_r_fr_MSTR (P)"(ComponentPose=@n72, bAlphaBoolEnabled=True, ConstraintWeights_0=1) {
    alphaInputType=Bool
    alphaBoolBlend {blendInTime=0.1, blendOption=Linear, blendOutTime=0.1}
    boneToModify {boneName=dyn_loop_r_fr_MSTR}
    constraintSetup [{targetBone {boneName=thigh_twist_01_r}}]
    constraintWeights [1]
   }
   n74 = "Constraint - dyn_loop_l_fr (P)"(ComponentPose=@n73, bAlphaBoolEnabled=True, ConstraintWeights_0=1) {
    alphaInputType=Bool
    alphaBoolBlend {blendInTime=0.1, blendOption=Linear, blendOutTime=0.1}
    boneToModify {boneName=dyn_loop_l_fr}
    constraintSetup [{targetBone {boneName=thigh_twist_01_l}}]
    constraintWeights [1]
   }
   n75 = "Constraint - dyn_loop_r_fr (P)"(ComponentPose=@n74, bAlphaBoolEnabled=True, ConstraintWeights_0=1) {
    alphaInputType=Bool
    alphaBoolBlend {blendInTime=0.1, blendOption=Linear, blendOutTime=0.1}
    boneToModify {boneName=dyn_loop_r_fr}
    constraintSetup [{targetBone {boneName=thigh_twist_01_r}}]
    constraintWeights [1]
   }
   n76 = "Constraint - dyn_l_front_square (P)"(ComponentPose=@n75, bAlphaBoolEnabled=True, ConstraintWeights_0=1) {
    alphaInputType=Bool
    alphaBoolBlend {blendInTime=0.1, blendOption=Linear, blendOutTime=0.1}
    boneToModify {boneName=dyn_l_front_square}
    constraintSetup [{targetBone {boneName=thigh_twist_01_l}}]
    constraintWeights [1]
   }
   n77 = "Constraint - dyn_r_front_square (P)"(ComponentPose=@n76, bAlphaBoolEnabled=True, ConstraintWeights_0=1) {
    alphaInputType=Bool
    alphaBoolBlend {blendInTime=0.1, blendOption=Linear, blendOutTime=0.1}
    boneToModify {boneName=dyn_r_front_square}
    constraintSetup [{targetBone {boneName=thigh_twist_01_r}}]
    constraintWeights [1]
   }
   n78 = "Constraint - dyn_l_front_ring_01 (P)"(ComponentPose=@n77, bAlphaBoolEnabled=True, ConstraintWeights_0=1) {
    alphaInputType=Bool
    alphaBoolBlend {blendInTime=0.1, blendOption=Linear, blendOutTime=0.1}
    boneToModify {boneName=dyn_l_front_ring_01}
    constraintSetup [{targetBone {boneName=thigh_twist_01_l}}]
    constraintWeights [1]
   }
   n79 = "Constraint - dyn_r_front_ring_01 (P)"(ComponentPose=@n78, bAlphaBoolEnabled=True, ConstraintWeights_0=1) {
    alphaInputType=Bool
    alphaBoolBlend {blendInTime=0.1, blendOption=Linear, blendOutTime=0.1}
    boneToModify {boneName=dyn_r_front_ring_01}
    constraintSetup [{targetBone {boneName=thigh_twist_01_r}}]
    constraintWeights [1]
   }
   n80 = "Constraint - dyn_l_front_ring_02 (P)"(ComponentPose=@n79, bAlphaBoolEnabled=True, ConstraintWeights_0=1) {
    alphaInputType=Bool
    alphaBoolBlend {blendInTime=0.1, blendOption=Linear, blendOutTime=0.1}
    boneToModify {boneName=dyn_l_front_ring_02}
    constraintSetup [{targetBone {boneName=thigh_twist_01_l}}]
    constraintWeights [1]
   }
   n81 = "Constraint - dyn_r_front_ring_02 (P)"(ComponentPose=@n80, bAlphaBoolEnabled=True, ConstraintWeights_0=1) {
    alphaInputType=Bool
    alphaBoolBlend {blendInTime=0.1, blendOption=Linear, blendOutTime=0.1}
    boneToModify {boneName=dyn_r_front_ring_02}
    constraintSetup [{targetBone {boneName=thigh_twist_01_r}}]
    constraintWeights [1]
   }
   n82 = Component To Local(ComponentPose=@n81)
   n83 = AND Boolean(A=Get bIsCrouching(), B=Get bRigidBodyEnabled())
   n84 = Anim Dynamics - Bone: dyn_necklace_01(
    ComponentPose=Local To Component(LocalPose=Blend Poses by bool(BlendPose_0=@n82, BlendPose_1=@n66, bActiveValue=@n83) {blendType=Linear})
    Alpha=1
    ExternalForce="0,0,0"
   ) {
    angularDampingOverride=1
    angularSpringConstant=1000
    bAngularSpring=true
    bGravityOverrideInSimSpace=true
    bOverrideAngularDamping=true
    bUseSphericalLimits=true
    gravityScale=0.35
    simulationSpace=RootRelative
    alphaBoolBlend {blendOption=Linear}
    alphaScaleBiasClamp {scale=0.7}
    boundBone {boneName=dyn_necklace_01}
    componentLinearAccScale=<0.4 0.4 0.46>
    componentLinearVelScale=<0.4 0.4 0.46>
    physicsBodyDefinitions [{
      sphereCollisionRadius=0
      boundBone {boneName=dyn_necklace_01}
      boxExtents=<10 20 20>
      constraintSetup {angularLimitsMax=<10 20 20>, angularLimitsMin=<-10 -20 -40>, angularTarget=<0 0 0>}
      localJointOffset=<5 0 0>
     }]
    sphericalLimits [{limitRadius=20, drivingBone {boneName=spine_05}, sphereLocalOffset=<0 -11 0>}]
   }
   n85 = Anim Dynamics - Bone: dyn_necklace_02(ComponentPose=@n84, Alpha=1, ExternalForce="0,0,0") {
    angularDampingOverride=0.999
    angularSpringConstant=1000
    bAngularSpring=true
    bGravityOverrideInSimSpace=true
    bOverrideAngularDamping=true
    bUseSphericalLimits=true
    gravityScale=0.35
    simulationSpace=RootRelative
    alphaBoolBlend {blendOption=Linear}
    alphaScaleBiasClamp {scale=0.7}
    boundBone {boneName=dyn_necklace_02}
    componentLinearAccScale=<0.4 0.4 0.46>
    componentLinearVelScale=<0.4 0.4 0.46>
    physicsBodyDefinitions [{
      sphereCollisionRadius=0
      boundBone {boneName=dyn_necklace_02}
      boxExtents=<8 15 15>
      constraintSetup {angularLimitsMax=<10 20 20>, angularLimitsMin=<-10 -20 -40>, angularTarget=<0 0 0>}
      localJointOffset=<5 0 0>
     }]
    sphericalLimits [{limitRadius=20, drivingBone {boneName=spine_05}, sphereLocalOffset=<0 -11 0>}]
   }
   n86 = Anim Dynamics - Bone: dyn_necklace_03(ComponentPose=@n85, Alpha=1, ExternalForce="0,0,0") {
    angularDampingOverride=1
    angularSpringConstant=1000
    bAngularSpring=true
    bGravityOverrideInSimSpace=true
    bOverrideAngularDamping=true
    bUseSphericalLimits=true
    gravityScale=0.35
    simulationSpace=RootRelative
    alphaBoolBlend {blendOption=Linear}
    alphaScaleBiasClamp {scale=0.7}
    boundBone {boneName=dyn_necklace_03}
    componentLinearAccScale=<0.4 0.4 0.46>
    componentLinearVelScale=<0.4 0.4 0.46>
    physicsBodyDefinitions [{
      sphereCollisionRadius=0
      boundBone {boneName=dyn_necklace_03}
      boxExtents=<10 20 20>
      constraintSetup {angularLimitsMax=<10 20 20>, angularLimitsMin=<-10 -20 -40>, angularTarget=<0 0 0>}
      localJointOffset=<5 0 0>
     }]
    sphericalLimits [{limitRadius=20, drivingBone {boneName=spine_05}, sphereLocalOffset=<0 -11 0>}]
   }
   n87 = Anim Dynamics - Bone: dyn_necklace_04(ComponentPose=@n86, Alpha=1, ExternalForce="0,0,0") {
    angularDampingOverride=1
    angularSpringConstant=1000
    bAngularSpring=true
    bGravityOverrideInSimSpace=true
    bOverrideAngularDamping=true
    bUseSphericalLimits=true
    gravityScale=0.35
    simulationSpace=RootRelative
    alphaBoolBlend {blendOption=Linear}
    alphaScaleBiasClamp {scale=0.7}
    boundBone {boneName=dyn_necklace_04}
    componentLinearAccScale=<0.4 0.4 0.46>
    componentLinearVelScale=<0.4 0.4 0.46>
    physicsBodyDefinitions [{
      sphereCollisionRadius=0
      boundBone {boneName=dyn_necklace_04}
      boxExtents=<5 10 10>
      constraintSetup {angularLimitsMax=<10 20 20>, angularLimitsMin=<-10 -20 -40>, angularTarget=<0 0 0>}
      localJointOffset=<5 0 0>
     }]
    sphericalLimits [{limitRadius=20, drivingBone {boneName=spine_05}, sphereLocalOffset=<0 -11 0>}]
   }
   n88 = "Transform (Modify) Bone - Bone: dyn_l_FR_Dress_MSTR"!(ComponentPose=@n87, bAlphaBoolEnabled:bool=True, Translation="7.379,-4.394,3.857", Rotation="-10,-0.016,0", Scale="1,1,1") -> (Pose) {
    alphaInputType=Bool
    rotationMode=BMM_Additive
    rotationSpace=BCS_BoneSpace
    scaleMode=BMM_Additive
    scaleSpace=BCS_BoneSpace
    translationMode=BMM_Additive
    translationSpace=BCS_BoneSpace
    alphaBoolBlend {blendOption=Linear}
    boneToModify {boneName=dyn_l_FR_Dress_MSTR}
    rotation {pitch=-10, yaw=-0.016}
    translation=<7.379 -4.394 3.857>
   }
   n89 = "Transform (Modify) Bone - Bone: dyn_l_Side_Dress_MSTR"!(ComponentPose=@n88, bAlphaBoolEnabled:bool=True, Translation="-3.592,11.027,3.255", Rotation="-0.023,-60.455,0.032", Scale="1,1,1") -> (Pose) {
    alphaInputType=Bool
    rotationMode=BMM_Additive
    rotationSpace=BCS_BoneSpace
    scaleMode=BMM_Additive
    scaleSpace=BCS_BoneSpace
    translationMode=BMM_Additive
    translationSpace=BCS_BoneSpace
    alphaBoolBlend {blendOption=Linear}
    boneToModify {boneName=dyn_l_Side_Dress_MSTR}
    rotation=<rot -0.023 -60.455 0.032>
    translation=<-3.592 11.027 3.255>
   }
   n90 = "Transform (Modify) Bone - Bone: dyn_l_dagger_MSTR"!(ComponentPose=@n89, bAlphaBoolEnabled:bool=True, Translation="7.603,1.126,-1.141", Rotation="7.695,-27.985,-37.148", Scale="1,1,1") -> (Pose) {
    alphaInputType=Bool
    rotationMode=BMM_Additive
    rotationSpace=BCS_BoneSpace
    scaleMode=BMM_Additive
    scaleSpace=BCS_BoneSpace
    translationMode=BMM_Additive
    translationSpace=BCS_BoneSpace
    alphaBoolBlend {blendOption=Linear}
    boneToModify {boneName=dyn_l_dagger_MSTR}
    rotation=<rot 7.695 -27.985 -37.148>
    translation=<7.603 1.126 -1.141>
   }
   n91 = "Constraint - dyn_l_FR_Dress (P)"(ComponentPose=@n90, Alpha=0.5, ConstraintWeights_0=1) {
    alphaBoolBlend {blendOption=Linear}
    boneToModify {boneName=dyn_l_FR_Dress}
    constraintSetup [{targetBone {boneName=dyn_loop_l_fr}}]
    constraintWeights [1]
   }
   n92 = "Transform (Modify) Bone - Bone: dyn_l_FR_Dress_MSTR"(ComponentPose=@n91, bAlphaBoolEnabled=True, Translation="0,0,0", Rotation="5,0,0", Scale="1,1,1") {
    alphaInputType=Bool
    rotationMode=BMM_Additive
    rotationSpace=BCS_BoneSpace
    scaleMode=BMM_Additive
    scaleSpace=BCS_BoneSpace
    translationMode=BMM_Additive
    translationSpace=BCS_BoneSpace
    alphaBoolBlend {blendInTime=0.1, blendOption=Linear, blendOutTime=0.1}
    boneToModify {boneName=dyn_l_FR_Dress_MSTR}
    rotation {pitch=5}
   }
   n93 = Anim Dynamics - Bone: dyn_l_wrist_zipper(ComponentPose=@n92, Alpha=1, ExternalForce="0,0,0") {
    angularDampingOverride=0.9999
    bGravityOverrideInSimSpace=true
    bOverrideAngularDamping=true
    simulationSpace=RootRelative
    alphaBoolBlend {blendOption=Linear}
    boundBone {boneName=dyn_l_wrist_zipper}
    physicsBodyDefinitions [{
      sphereCollisionRadius=0
      boundBone {boneName=dyn_l_wrist_zipper}
      boxExtents=<6 13 13>
      constraintSetup {angularLimitsMax=<2 75 45>, angularLimitsMin=<-2 -10 -45>, angularTarget=<0 0 0>}
      localJointOffset=<1.5 -0.05 0.1>
     }]
   }
   n94 = Anim Dynamics - Bone: dyn_l_boot_zipper(ComponentPose=@n93, Alpha=1, ExternalForce="0,0,0") {
    angularDampingOverride=0.9999
    bGravityOverrideInSimSpace=true
    bOverrideAngularDamping=true
    simulationSpace=RootRelative
    alphaBoolBlend {blendOption=Linear}
    boundBone {boneName=dyn_l_boot_zipper}
    physicsBodyDefinitions [{
      sphereCollisionRadius=0
      boundBone {boneName=dyn_l_boot_zipper}
      boxExtents=<6 13 13>
      constraintSetup {angularLimitsMax=<2 15 45>, angularLimitsMin=<-2 -60 -45>, angularTarget=<0 0 0>}
      localJointOffset=<1.5 -0.05 0.1>
     }]
   }
   n95 = Anim Dynamics - Bone: dyn_r_boot_zipper(ComponentPose=@n94, Alpha=1, ExternalForce="0,0,0") {
    angularDampingOverride=0.9999
    bGravityOverrideInSimSpace=true
    bOverrideAngularDamping=true
    simulationSpace=RootRelative
    alphaBoolBlend {blendOption=Linear}
    boundBone {boneName=dyn_r_boot_zipper}
    physicsBodyDefinitions [{
      sphereCollisionRadius=0
      boundBone {boneName=dyn_r_boot_zipper}
      boxExtents=<6 13 13>
      constraintSetup {angularLimitsMax=<2 10 45>, angularLimitsMin=<-2 -75 -45>, angularTarget=<0 0 0>}
      localJointOffset=<1.5 -0.05 0.1>
     }]
   }
   n96 = "Transform (Modify) Bone - Bone: dyn_loop_l_fr_MSTR"(ComponentPose=@n95, bAlphaBoolEnabled=True, Translation="0,0,0", Rotation="15.136,-0.031,6.435", Scale="1,1,1") {
    alphaInputType=Bool
    rotationMode=BMM_Additive
    rotationSpace=BCS_BoneSpace
    scaleMode=BMM_Additive
    scaleSpace=BCS_BoneSpace
    translationMode=BMM_Additive
    translationSpace=BCS_BoneSpace
    alphaBoolBlend {blendInTime=0.1, blendOption=Linear, blendOutTime=0.1}
    boneToModify {boneName=dyn_loop_l_fr_MSTR}
    rotation=<rot 15.136 -0.031 6.435>
   }
   n97 = "Transform (Modify) Bone - Bone: dyn_loop_r_fr_MSTR"(ComponentPose=@n96, bAlphaBoolEnabled=True, Translation="0.334,-0.347,1.732", Rotation="-18.418,-21.755,-2.353", Scale="1,1,1") {
    alphaInputType=Bool
    rotationMode=BMM_Additive
    rotationSpace=BCS_BoneSpace
    scaleMode=BMM_Additive
    scaleSpace=BCS_BoneSpace
    translationMode=BMM_Additive
    translationSpace=BCS_BoneSpace
    alphaBoolBlend {blendInTime=0.1, blendOption=Linear, blendOutTime=0.1}
    boneToModify {boneName=dyn_loop_r_fr_MSTR}
    rotation=<rot -18.418 -21.755 -2.353>
    translation=<0.3341 -0.347 1.7323>
   }
   n98 = "Transform (Modify) Bone - Bone: dyn_loop_l_MSTR"(ComponentPose=@n97, bAlphaBoolEnabled=True, Translation="0,0,0", Rotation="0.414,11.393,33.145", Scale="1,1,1") {
    alphaInputType=Bool
    rotationMode=BMM_Additive
    rotationSpace=BCS_BoneSpace
    scaleMode=BMM_Additive
    scaleSpace=BCS_BoneSpace
    translationMode=BMM_Additive
    translationSpace=BCS_BoneSpace
    alphaBoolBlend {blendInTime=0.1, blendOption=Linear, blendOutTime=0.1}
    boneToModify {boneName=dyn_loop_l_MSTR}
    rotation=<rot 0.414 11.393 33.145>
   }
   n99 = "Transform (Modify) Bone - Bone: dyn_loop_r_MSTR"(ComponentPose=@n98, bAlphaBoolEnabled=True, Translation="0,0,0", Rotation="-18.947,-0.390,-35.084", Scale="1,1,1") {
    alphaInputType=Bool
    rotationMode=BMM_Additive
    rotationSpace=BCS_BoneSpace
    scaleMode=BMM_Additive
    scaleSpace=BCS_BoneSpace
    translationMode=BMM_Additive
    translationSpace=BCS_BoneSpace
    alphaBoolBlend {blendInTime=0.1, blendOption=Linear, blendOutTime=0.1}
    boneToModify {boneName=dyn_loop_r_MSTR}
    rotation=<rot -18.947 -0.39 -35.084>
   }
   n100 = "Transform (Modify) Bone - Bone: dyn_back_strap_l_01"!(ComponentPose=@n99, bAlphaBoolEnabled:bool=True, Translation="0,0,0", Rotation="13.666,-8.062,-9.408", Scale="1,1,1") -> (Pose) {
    alphaInputType=Bool
    rotationMode=BMM_Additive
    rotationSpace=BCS_BoneSpace
    scaleMode=BMM_Additive
    scaleSpace=BCS_BoneSpace
    translationMode=BMM_Additive
    translationSpace=BCS_BoneSpace
    alphaBoolBlend {blendInTime=0.1, blendOption=Linear, blendOutTime=0.1}
    boneToModify {boneName=dyn_back_strap_l_01}
    rotation=<rot 13.666 -8.062 -9.408>
   }
   n101 = "Transform (Modify) Bone - Bone: dyn_back_strap_r_MSTR"!(ComponentPose=@n100, bAlphaBoolEnabled:bool=True, Translation="0,0,0", Rotation="-24.485,-9.862,2.024", Scale="1,1,1") -> (Pose) {
    alphaInputType=Bool
    rotationMode=BMM_Additive
    rotationSpace=BCS_BoneSpace
    scaleMode=BMM_Additive
    scaleSpace=BCS_BoneSpace
    translationMode=BMM_Additive
    translationSpace=BCS_BoneSpace
    alphaBoolBlend {blendInTime=0.1, blendOption=Linear, blendOutTime=0.1}
    boneToModify {boneName=dyn_back_strap_r_MSTR}
    rotation=<rot -24.485 -9.862 2.024>
   }
   n102 = "Transform (Modify) Bone - Bone: dyn_loop_l_bk_MSTR"!(ComponentPose=@n101, bAlphaBoolEnabled:bool=True, Translation="0,0,0", Rotation="-15.724,3.803,-13.730", Scale="1,1,1") -> (Pose) {
    alphaInputType=Bool
    rotationMode=BMM_Additive
    rotationSpace=BCS_BoneSpace
    scaleMode=BMM_Additive
    scaleSpace=BCS_BoneSpace
    translationMode=BMM_Additive
    translationSpace=BCS_BoneSpace
    alphaBoolBlend {blendInTime=0.1, blendOption=Linear, blendOutTime=0.1}
    boneToModify {boneName=dyn_loop_l_bk_MSTR}
    rotation=<rot -15.7241 3.803 -13.7301>
   }
   n103 = "Transform (Modify) Bone - Bone: dyn_loop_r_bk_MSTR"!(ComponentPose=@n102, bAlphaBoolEnabled:bool=True, Translation="0,0,0", Rotation="-2.711,-1.519,12.223", Scale="1,1,1") -> (Pose) {
    alphaInputType=Bool
    rotationMode=BMM_Additive
    rotationSpace=BCS_BoneSpace
    scaleMode=BMM_Additive
    scaleSpace=BCS_BoneSpace
    translationMode=BMM_Additive
    translationSpace=BCS_BoneSpace
    alphaBoolBlend {blendInTime=0.1, blendOption=Linear, blendOutTime=0.1}
    boneToModify {boneName=dyn_loop_r_bk_MSTR}
    rotation=<rot -2.711 -1.519 12.223>
   }
   n104 = "Transform (Modify) Bone - Bone: dyn_loop_l_fr_MSTR"(ComponentPose=@n103, bAlphaBoolEnabled=True, Translation="0,0,0", Rotation="4.317,-12.736,-0.000", Scale="1,1,1") {
    alphaInputType=Bool
    rotationMode=BMM_Additive
    rotationSpace=BCS_BoneSpace
    scaleMode=BMM_Additive
    scaleSpace=BCS_BoneSpace
    translationMode=BMM_Additive
    translationSpace=BCS_BoneSpace
    alphaBoolBlend {blendInTime=0.1, blendOption=Linear, blendOutTime=0.1}
    boneToModify {boneName=dyn_loop_l_fr_MSTR}
    rotation=<rot 4.317 -12.736 0>
   }
   n105 = "Transform (Modify) Bone - Bone: dyn_loop_r_fr_MSTR"(ComponentPose=@n104, bAlphaBoolEnabled=True, Translation="0,0,0", Rotation="4.810,4.884,0.940", Scale="1,1,1") {
    alphaInputType=Bool
    rotationMode=BMM_Additive
    rotationSpace=BCS_BoneSpace
    scaleMode=BMM_Additive
    scaleSpace=BCS_BoneSpace
    translationMode=BMM_Additive
    translationSpace=BCS_BoneSpace
    alphaBoolBlend {blendInTime=0.1, blendOption=Linear, blendOutTime=0.1}
    boneToModify {boneName=dyn_loop_r_fr_MSTR}
    rotation=<rot 4.81 4.884 0.94>
   }
   n106 = "Transform (Modify) Bone - Bone: dyn_loop_l_MSTR"(ComponentPose=@n105, bAlphaBoolEnabled=True, Translation="0,0,0", Rotation="15.916,0.000,-0.000", Scale="1,1,1") {
    alphaInputType=Bool
    rotationMode=BMM_Additive
    rotationSpace=BCS_BoneSpace
    scaleMode=BMM_Additive
    scaleSpace=BCS_BoneSpace
    translationMode=BMM_Additive
    translationSpace=BCS_BoneSpace
    alphaBoolBlend {blendInTime=0.1, blendOption=Linear, blendOutTime=0.1}
    boneToModify {boneName=dyn_loop_l_MSTR}
    rotation=<rot 15.916 0.0001 -0.0001>
   }
   n107 = "Transform (Modify) Bone - Bone: dyn_loop_r_MSTR"(ComponentPose=@n106, bAlphaBoolEnabled=True, Translation="0,0,0", Rotation="-0.000,-15.281,0.000", Scale="1,1,1") {
    alphaInputType=Bool
    rotationMode=BMM_Additive
    rotationSpace=BCS_BoneSpace
    scaleMode=BMM_Additive
    scaleSpace=BCS_BoneSpace
    translationMode=BMM_Additive
    translationSpace=BCS_BoneSpace
    alphaBoolBlend {blendInTime=0.1, blendOption=Linear, blendOutTime=0.1}
    boneToModify {boneName=dyn_loop_r_MSTR}
    rotation=<rot 0 -15.281 0>
   }
   n108 = "Transform (Modify) Bone - Bone: dyn_back_strap_l_01"(ComponentPose=@n107, bAlphaBoolEnabled=True, Translation="0,0,0", Rotation="6.517,-6.276,-0.000", Scale="1,1,1") {
    alphaInputType=Bool
    rotationMode=BMM_Additive
    rotationSpace=BCS_BoneSpace
    scaleMode=BMM_Additive
    scaleSpace=BCS_BoneSpace
    translationMode=BMM_Additive
    translationSpace=BCS_BoneSpace
    alphaBoolBlend {blendInTime=0.1, blendOption=Linear, blendOutTime=0.1}
    boneToModify {boneName=dyn_back_strap_l_01}
    rotation=<rot 6.517 -6.276 0>
   }
   n109 = "Transform (Modify) Bone - Bone: dyn_back_strap_r_MSTR"(ComponentPose=@n108, bAlphaBoolEnabled=True, Translation="0,0,0", Rotation="-3.069,-8.498,0.000", Scale="1,1,1") {
    alphaInputType=Bool
    rotationMode=BMM_Additive
    rotationSpace=BCS_BoneSpace
    scaleMode=BMM_Additive
    scaleSpace=BCS_BoneSpace
    translationMode=BMM_Additive
    translationSpace=BCS_BoneSpace
    alphaBoolBlend {blendInTime=0.1, blendOption=Linear, blendOutTime=0.1}
    boneToModify {boneName=dyn_back_strap_r_MSTR}
    rotation=<rot -3.069 -8.498 0>
   }
   n110 = "Transform (Modify) Bone - Bone: dyn_loop_l_bk_MSTR"(ComponentPose=@n109, bAlphaBoolEnabled=True, Translation="0,0,0", Rotation="0,0,0", Scale="1,1,1") {
    alphaInputType=Bool
    rotationMode=BMM_Additive
    rotationSpace=BCS_BoneSpace
    scaleMode=BMM_Additive
    scaleSpace=BCS_BoneSpace
    translationMode=BMM_Additive
    translationSpace=BCS_BoneSpace
    alphaBoolBlend {blendInTime=0.1, blendOption=Linear, blendOutTime=0.1}
    boneToModify {boneName=dyn_loop_l_bk_MSTR}
   }
   n111 = "Transform (Modify) Bone - Bone: dyn_loop_r_bk_MSTR"(ComponentPose=@n110, bAlphaBoolEnabled=True, Translation="0,0,0", Rotation="0,0,0", Scale="1,1,1") {
    alphaInputType=Bool
    rotationMode=BMM_Additive
    rotationSpace=BCS_BoneSpace
    scaleMode=BMM_Additive
    scaleSpace=BCS_BoneSpace
    translationMode=BMM_Additive
    translationSpace=BCS_BoneSpace
    alphaBoolBlend {blendInTime=0.1, blendOption=Linear, blendOutTime=0.1}
    boneToModify {boneName=dyn_loop_r_bk_MSTR}
   }
   n112 = "Constraint - dyn_loop_r_fr (R)"!(ComponentPose=@n111, bAlphaBoolEnabled:bool=True, ConstraintWeights_0=1) -> (Pose) {
    alphaInputType=Bool
    alphaBoolBlend {blendOption=Linear}
    boneToModify {boneName=dyn_loop_r_fr}
    constraintSetup [{transformType=Rotation, targetBone {boneName=thigh_r}}]
    constraintWeights [1]
   }
   n113 = "Transform (Modify) Bone - Bone: dyn_loop_r_fr_MSTR"(ComponentPose=@n112, bAlphaBoolEnabled=True, Translation="5.600,3.202,-4.750", Rotation="-0.000,16.949,1.654", Scale="1,1,1") {
    alphaInputType=Bool
    rotationMode=BMM_Additive
    rotationSpace=BCS_BoneSpace
    scaleMode=BMM_Additive
    scaleSpace=BCS_BoneSpace
    translationMode=BMM_Additive
    translationSpace=BCS_BoneSpace
    alphaBoolBlend {blendInTime=0.1, blendOption=Linear, blendOutTime=0.1}
    boneToModify {boneName=dyn_loop_r_fr_MSTR}
    rotation=<rot 0 16.949 1.654>
    translation=<5.6 3.202 -4.75>
   }
   Output Pose(Result=Component To Local(ComponentPose=@n113)) {layerGroup=None}
   n114 = Control Rig(Source=Copy Pose From Mesh() {bCopyCurves=true, bUseAttachedParent=true}) {
    alphaBoolBlend {blendOption=Linear}
    controlRigAssetReference {blueprintRigClass=&/F_MED_Sunlit_DeformRig.F_MED_Sunlit_DeformRig_C}
   }
   n115 = "Constraint - dyn_necklace_MSTR (PPPPPP)"!(
    ComponentPose=Local To Component(LocalPose=@n114)
    Alpha=1
    ConstraintWeights_0=0.0104
    ConstraintWeights_1:float=0.144
    ConstraintWeights_2:float=0.1265
    ConstraintWeights_3:float=0.1414
    ConstraintWeights_4:float=0.0579
    ConstraintWeights_5:float=0.5198
   ) -> (Pose) {
    alphaBoolBlend {blendOption=Linear}
    boneToModify {boneName=dyn_necklace_MSTR}
    constraintSetup [
     {targetBone {boneName=spine_03}}
     {targetBone {boneName=neck_02}}
     {targetBone {boneName=spine_04}}
     {targetBone {boneName=neck_01}}
     {targetBone {boneName=head}}
     {targetBone {boneName=spine_05}}
    ]
    constraintWeights [1,1,1,1,1,1]
   }
   n116 = "Constraint - dyn_loop_l_fr_MSTR (PPPPPPP)"!(
    ComponentPose=@n115
    Alpha=1
    ConstraintWeights_0=0.8493
    ConstraintWeights_1:float=0.1044
    ConstraintWeights_2:float=0.0063
    ConstraintWeights_3:float=0.0005
    ConstraintWeights_4:float=0.0077
    ConstraintWeights_5:float=0.0009
    ConstraintWeights_6:float=0.0308
   ) -> (Pose) {
    alphaBoolBlend {blendOption=Linear}
    boneToModify {boneName=dyn_loop_l_fr_MSTR}
    constraintSetup [
     {targetBone {boneName=pelvis}}
     {targetBone {boneName=spine_02}}
     {targetBone {boneName=spine_04}}
     {targetBone {boneName=spine_01}}
     {targetBone {boneName=thigh_twist_01_l}}
     {targetBone {boneName=spine_03}}
     {targetBone {boneName=deform_groin_l}}
    ]
    constraintWeights [1,1,1,1,1,1,1]
   }
   n117 = "Constraint - dyn_loop_r_fr_MSTR (PPPPPPP)"!(
    ComponentPose=@n116
    Alpha=1
    ConstraintWeights_0=0.8524
    ConstraintWeights_1:float=0.0079
    ConstraintWeights_2:float=0.1012
    ConstraintWeights_3:float=0.0057
    ConstraintWeights_4:float=0.0004
    ConstraintWeights_5:float=0.0006
    ConstraintWeights_6:float=0.0319
   ) -> (Pose) {
    alphaBoolBlend {blendOption=Linear}
    boneToModify {boneName=dyn_loop_r_fr_MSTR}
    constraintSetup [
     {targetBone {boneName=pelvis}}
     {targetBone {boneName=thigh_twist_01_r}}
     {targetBone {boneName=spine_02}}
     {targetBone {boneName=spine_04}}
     {targetBone {boneName=spine_01}}
     {targetBone {boneName=spine_03}}
     {targetBone {boneName=deform_groin_r}}
    ]
    constraintWeights [1,1,1,1,1,1,1]
   }
   n118 = "Constraint - dyn_l_BK_Dress_MSTR (PPPPP)"!(
    ComponentPose=@n117
    Alpha=1
    ConstraintWeights_0=0.0001
    ConstraintWeights_1:float=0.0081
    ConstraintWeights_2:float=0.0423
    ConstraintWeights_3:float=0.9277
    ConstraintWeights_4:float=0.0219
   ) -> (Pose) {
    alphaBoolBlend {blendOption=Linear}
    boneToModify {boneName=dyn_l_BK_Dress_MSTR}
    constraintSetup [
     {targetBone {boneName=thigh_twist_01_l}}
     {targetBone {boneName=spine_01}}
     {targetBone {boneName=deform_glute_l}}
     {targetBone {boneName=pelvis}}
     {targetBone {boneName=spine_02}}
    ]
    constraintWeights [1,1,1,1,1]
   }
   n119 = "Constraint - dyn_r_BK_Dress_MSTR (PPPPP)"!(
    ComponentPose=@n118
    Alpha=1
    ConstraintWeights_0=0.0081
    ConstraintWeights_1:float=0.9277
    ConstraintWeights_2:float=0.0423
    ConstraintWeights_3:float=0.0001
    ConstraintWeights_4:float=0.0219
   ) -> (Pose) {
    alphaBoolBlend {blendOption=Linear}
    boneToModify {boneName=dyn_r_BK_Dress_MSTR}
    constraintSetup [
     {targetBone {boneName=spine_01}}
     {targetBone {boneName=pelvis}}
     {targetBone {boneName=deform_glute_r}}
     {targetBone {boneName=thigh_twist_01_r}}
     {targetBone {boneName=spine_02}}
    ]
    constraintWeights [1,1,1,1,1]
   }
   n120 = "Constraint - dyn_back_strap_l_MSTR (PPPPP)"!(
    ComponentPose=@n119
    Alpha=1
    ConstraintWeights_0=0.1524
    ConstraintWeights_1:float=0.3939
    ConstraintWeights_2:float=0.0613
    ConstraintWeights_3:float=0.2646
    ConstraintWeights_4:float=0.1278
   ) -> (Pose) {
    alphaBoolBlend {blendOption=Linear}
    boneToModify {boneName=dyn_back_strap_l_MSTR}
    constraintSetup [
     {targetBone {boneName=pelvis}}
     {targetBone {boneName=spine_02}}
     {targetBone {boneName=spine_04}}
     {targetBone {boneName=spine_01}}
     {targetBone {boneName=spine_03}}
    ]
    constraintWeights [1,1,1,1,1]
   }
   n121 = "Constraint - dyn_back_strap_r_MSTR (PPPPP)"!(
    ComponentPose=@n120
    Alpha=1
    ConstraintWeights_0=0.1548
    ConstraintWeights_1:float=0.3911
    ConstraintWeights_2:float=0.0621
    ConstraintWeights_3:float=0.2645
    ConstraintWeights_4:float=0.1275
   ) -> (Pose) {
    alphaBoolBlend {blendOption=Linear}
    boneToModify {boneName=dyn_back_strap_r_MSTR}
    constraintSetup [
     {targetBone {boneName=pelvis}}
     {targetBone {boneName=spine_02}}
     {targetBone {boneName=spine_04}}
     {targetBone {boneName=spine_01}}
     {targetBone {boneName=spine_03}}
    ]
    constraintWeights [1,1,1,1,1]
   }
   n122 = Get bIsSliding()
   n123 = "Transform (Modify) Bone - Bone: dyn_back_strap_l_MSTR"!(ComponentPose=@n121, bAlphaBoolEnabled:bool=@n122, Translation="0,0,0", Rotation="25,-30,0.026", Scale="1,1,1") -> (Pose) {
    alphaInputType=Bool
    rotationMode=BMM_Additive
    rotationSpace=BCS_BoneSpace
    translationMode=BMM_Additive
    translationSpace=BCS_BoneSpace
    alphaBoolBlend {blendInTime=0.1, blendOption=Linear, blendOutTime=0.1}
    boneToModify {boneName=dyn_back_strap_l_MSTR}
    rotation=<rot 25 -30 0.026>
   }
   n124 = "Transform (Modify) Bone - Bone: dyn_back_strap_r_MSTR"(ComponentPose=@n123, bAlphaBoolEnabled=@n122, Translation="0,0,0", Rotation="0,-30,0", Scale="1,1,1") {
    alphaInputType=Bool
    rotationMode=BMM_Additive
    rotationSpace=BCS_BoneSpace
    translationMode=BMM_Additive
    translationSpace=BCS_BoneSpace
    alphaBoolBlend {blendInTime=0.1, blendOption=Linear, blendOutTime=0.1}
    boneToModify {boneName=dyn_back_strap_r_MSTR}
    rotation {yaw=-30}
   }
   n125 = "Transform (Modify) Bone - Bone: dyn_loop_l_fr_MSTR"(ComponentPose=@n124, bAlphaBoolEnabled=True, Translation="0,0,0", Rotation="49.244,-99.442,-108.332", Scale="1,1,1") {
    alphaInputType=Bool
    rotationMode=BMM_Additive
    rotationSpace=BCS_BoneSpace
    scaleMode=BMM_Additive
    scaleSpace=BCS_BoneSpace
    translationMode=BMM_Additive
    translationSpace=BCS_BoneSpace
    alphaBoolBlend {blendInTime=0.1, blendOption=Linear, blendOutTime=0.1}
    boneToModify {boneName=dyn_loop_l_fr_MSTR}
    rotation=<rot 49.244 -99.442 -108.332>
   }
   n126 = "Transform (Modify) Bone - Bone: dyn_loop_l_MSTR"(ComponentPose=@n125, bAlphaBoolEnabled=True, Translation="3.924,7.469,6.838", Rotation="-15.632,-55.977,9.174", Scale="1,1,1") {
    alphaInputType=Bool
    rotationMode=BMM_Additive
    rotationSpace=BCS_BoneSpace
    scaleMode=BMM_Additive
    scaleSpace=BCS_BoneSpace
    translationMode=BMM_Additive
    translationSpace=BCS_BoneSpace
    alphaBoolBlend {blendInTime=0.1, blendOption=Linear, blendOutTime=0.1}
    boneToModify {boneName=dyn_loop_l_MSTR}
    rotation=<rot -15.632 -55.977 9.174>
    translation=<3.924 7.469 6.838>
   }
   n127 = "Constraint - dyn_loop_r_fr_MSTR (R)"(ComponentPose=@n126, Alpha=0.5, ConstraintWeights_0=1) {
    alphaBoolBlend {blendOption=Linear}
    boneToModify {boneName=dyn_loop_r_fr_MSTR}
    constraintSetup [{transformType=Rotation, targetBone {boneName=thigh_r}}]
    constraintWeights [1]
   }
   n128 = "Constraint - dyn_r_BK_Dress_MSTR (PPPPPPPP)"!(
    ComponentPose=@n127
    Alpha=1
    ConstraintWeights_0=0.0354
    ConstraintWeights_1:float=0.7282
    ConstraintWeights_2:float=0.0114
    ConstraintWeights_3:float=0.0072
    ConstraintWeights_4:float=0.1231
    ConstraintWeights_5:float=0.0097
    ConstraintWeights_6:float=0.0536
    ConstraintWeights_7:float=0.0314
   ) -> (Pose) {
    alphaBoolBlend {blendOption=Linear}
    boneToModify {boneName=dyn_r_BK_Dress_MSTR}
    constraintSetup [
     {targetBone {boneName=thigh_twist_01_r}}
     {targetBone {boneName=pelvis}}
     {targetBone {boneName=spine_03}}
     {targetBone {boneName=deform_glute_l}}
     {targetBone {boneName=spine_02}}
     {targetBone {boneName=spine_04}}
     {targetBone {boneName=spine_01}}
     {targetBone {boneName=deform_glute_r}}
    ]
    constraintWeights [1,1,1,1,1,1,1,1]
   }
   Save cached pose 'PreSim'(Pose=Component To Local(ComponentPose=@n128)) {cacheName=PreSim}
  ]}
 EventGraph {nodes [
   Event BlueprintUpdateAnimation()
   > n1 = Sequence!(execute) -> (then_0, then_1, then_2:Exec, then_3:Exec, then_4:Exec, then_5:Exec, then_6:Exec, then_7:Exec, then_8:Exec)
   n2 = Branch(execute=@n1.then_1, Condition=OR Boolean!(A=Get bIsJumping(), B=Get bIsFalling(), C:bool=Get bIsDBNO()) -> (ReturnValue))
   > Set FloorCOL(FloorCOL:bool=true) -> (Output_Get:bool)
   Set FloorCOL(execute=@n2.else, FloorCOL=false)
   n3 = Branch(execute=@n1.then_2, Condition=OR Boolean(A=Get bIsJumping(), B=Get bIsFalling()))
   > Set IsJumpingNew(IsJumpingNew:bool=true) -> (Output_Get:bool)
   Set IsJumpingNew(execute=@n3.else, IsJumpingNew=false)
   n4 = Get bIsParachuteOpened()
   n5 = Branch(execute=@n1.then_3, Condition=AND Boolean(A=Get bIsSkyDiving(), B=NOT Boolean(A=@n4)))
   > Set IsSkydivingNew(IsSkydivingNew:bool=true) -> (Output_Get:bool)
   Set IsSkydivingNew(execute=@n5.else, IsSkydivingNew=false)
   n6 = Get skeletal_mesh() -> (skeletal_mesh:Skeletal Mesh Component Object Reference)
   Set ThighRotL(
    execute=@n1.then_5
    ThighRotL:double=GetAlphaFromReferenceTransformDelta(SkeletalMeshComponent=@n6, BoneName=thigh_l, BaseName=pelvis, InputRange=(X=-20,Y=-100), TwistAxis="1,0,0")
   ) -> (Output_Get:double)
   > Set ThighRotR(ThighRotR:double=GetAlphaFromReferenceTransformDelta(SkeletalMeshComponent=@n6, BoneName=thigh_r, BaseName=pelvis, InputRange=(X=-20,Y=-100), TwistAxis="1,0,0")) -> (Output_Get:double)
   > n7 = PlayerMovement(AnimInstance=Self-Reference(), XDotProductCoefficient=1, YDotProductCoefficient=1, ZDotProductCoefficient=1)
   > Set BwdMovement(BwdMovement:float=MapRangeClamped(Value=@n7.YMovement, InRangeA=0, InRangeB=-1, OutRangeA=1, OutRangeB=0.4)) -> (Output_Get:float)
   > n8 = Branch(Condition=Get bIsSkyDiving())
   > n9 = Branch(Condition="float > float"(A=Get PawnVelocityZ(), B=50) {is=title K2Node_CallFunction})
   > Set BwdMovement(BwdMovement=float * float(A=Get BwdMovement(), B=0))
   Set BwdMovement(execute=[@n9.else, @n8.else], BwdMovement=float * float(A=Get BwdMovement(), B=1))
   n10 = Branch(execute=@n1.then_7, Condition=NOT Boolean(A=Get bIsInFrontEnd()))
   > Set NotFrontEnd(NotFrontEnd:bool=true) -> (Output_Get:bool)
   Set NotFrontEnd(execute=@n10.else, NotFrontEnd=false)
   n11 = Branch(execute=@n1.then_8, Condition=Get bIsSkyDiving())
   > Set RelaxationAlpha(RelaxationAlpha:double=1) -> (Output_Get:double)
   Set RelaxationAlpha(execute=@n11.else, RelaxationAlpha=3)
   n12 = Branch(execute=@n1.then_6, Condition=AND Boolean(A=Get bIsCrouching(), B=Get bIsAccelerating2D()))
   > Set CrouchWalking(CrouchWalking:bool=true) -> (Output_Get:bool)
   Set CrouchWalking(execute=@n12.else, CrouchWalking=false)
   n13 = Gravity Override State Function DEPRECATED(
    execute=@n1.then_0
    CharacterPartAnimInstance=Self-Reference()
    GameTimeInSeconds=GetGameTimeInSeconds()
    GravityOverrideParams=Get GravityOverrideParams() -> (GravityOverrideParams:Gravity Override Params Struct Structure)
   )
   > Set GravityOverride(GravityOverride=@n13)
   n14 = Get bIsMoving2D()
   n15 = Branch(execute=@n1.then_4, Condition=AND Boolean(A=Get bIsCrouching(), B=NOT Boolean(A=@n14)))
   > Set still_crouch(still_crouch:bool=true) -> (Output_Get:bool)
   Set still_crouch(execute=@n15.else, still_crouch=false)
   Event BlueprintInitializeAnimation()
   > Set skeletal_mesh(skeletal_mesh:Skeletal Mesh Component Object Reference=GetOwnerSkeletalMeshComponent()) -> (Output_Get:Skeletal Mesh Component Object Reference)
   n16 = TryGetPawnOwner()
   > Is Valid(InputObject=@n16)
   > n17 = Cast To FortPlayerPawnAthena(Object=@n16)
   > Set playerPawn(playerPawn:Fort Player Pawn Athena Object Reference=@n17) -> (Output_Get:Fort Player Pawn Athena Object Reference)
  ]}
]
variables [
 BwdMovement {type=number}
 CrouchWalking {type=boolean}
 FloorCOL {type=boolean}
 GravityOverride {type=Vector}
 GravityOverrideParams {
  type=GravityOverrideParamsStruct
  default {skydive_FloatingDown {gravityOverride=<0 -700 1500>}, skydive_GoingUp {gravityOverride=<0 -300 300>}}
 }
 IsJumpingNew {type=boolean}
 IsSkydivingNew {type=boolean}
 NotFrontEnd {type=boolean}
 playerPawn {type=FortPlayerPawnAthena}
 RelaxationAlpha {type=number}
 skeletal_mesh {type=SkeletalMeshComponent}
 still_crouch {type=boolean}
 ThighRotL {type=number}
 ThighRotR {type=number}
]
